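# Physics asset for the FE mesh of the F_MED_Raspberry female player body, covering its dangling zippers, necklace and tassels.
# Pelvis, spine and neck shapes form the torso base, linked by free-rotation constraints.
# Small MSTR or Base bodies anchor accessories to spine or pelvis, freely rotating except the zero-degree back-tassel roots.
# Child bodies hang from these with angular limits, forming short chains for the necklace and two back tassels.
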PhysicsAsset /BRCosmetics/Characters/Player/Female/Medium/Bodies/F_MED_Raspberry/Meshes/F_MED_Raspberry_FE_Physics
bodies {
 dyn_BTMZipper [{
   shapeType=Box
   extentX=7.7827
   extentY=3.1097
   extentZ=7.3131
   shapeName=dyn_TopZipper_box
   center=<3.556 0.0223 0.2101>
   rotation=<rot 0 90.1008 -86.7327>
  }]
 dyn_BTMZipper_MSTR [{shapeType=Capsule, length=0.505, radius=0.505, shapeName=dyn_BTMZipper_MSTR_capsule, rotation=<rot 0 -90 0>}]
 dyn_TopZipper [{
   shapeType=Box
   extentX=1.3188
   extentY=0.9583
   extentZ=2.2378
   shapeName=dyn_TopZipper_box
   center=<1.1562 0.0181 0.0731>
   rotation=<rot 0 90.1008 -86.7327>
  }]
 dyn_TopZipper_MSTR [{shapeType=Capsule, length=0.505, radius=0.505, shapeName=dyn_TopZipper_MSTR_capsule, rotation=<rot 0 -90 0>}]
 dyn_necklaceMain [{
   shapeType=Box
   extentX=1.01
   extentY=1.01
   extentZ=1.01
   shapeName=dyn_necklaceMain_box
   center=<3.9835 0.1013 4.57>
   rotation=<rot 0 -90.0403 -70.1331>
  }]
 dyn_necklaceMain_Base [{shapeType=Box, extentX=1.01, extentY=1.01, extentZ=1.01, shapeName=dyn_necklaceMain_Base_box, rotation=<rot 0 -90 0>}]
 dyn_necklace_01 [{shapeType=Box, extentX=9.1405, extentY=4.7789, extentZ=16.1211, shapeName=dyn_necklace_01_box, center=<1.0396 -1 0>}]
 dyn_necklace_02 [{shapeType=Box, extentX=9.1405, extentY=4.7789, extentZ=16.1211, shapeName=dyn_necklace_02_box, center=<1.0396 -1 0>}]
 dyn_tassel_back_01_01_b [{shapeType=Capsule, length=3.7375, radius=0.5, shapeName=dyn_tassel_back_01_01_b_capsule, center=<2.2548 0 0.2424>, rotation=<rot 90 0 0>}]
 dyn_tassel_back_01_02_b [{
   shapeType=Capsule
   length=3.2315
   radius=0.5
   shapeName=dyn_tassel_back_01_02_b_capsule
   center=<2.0697 0 -0.2165>
   rotation=<rot 79.6984 0 0>
  }]
 dyn_tassel_back_01_a [{shapeType=Capsule, length=4.4, radius=0.5, shapeName=dyn_tassel_back_01_a_capsule, center=<1.9922 0 0.2>, rotation=<rot 87.2943 -180 -180>}]
 dyn_tassel_back_01_a_MSTR [{shapeType=Capsule, length=0.505, radius=0.505, shapeName=dyn_tassel_back_01_a_MSTR_capsule, rotation=<rot 90 0 0>}]
 dyn_tassel_back_01_b_MSTR [{shapeType=Capsule, length=0.505, radius=0.505, shapeName=dyn_tassel_back_01_b_MSTR_capsule, rotation=<rot 90 0 0>}]
 dyn_tassel_back_02_a [{shapeType=Capsule, length=3.6152, radius=0.8723, shapeName=dyn_tassel_back_02_a_capsule, center=<2.5 0 -0.26>, rotation=<rot 80 0 0>}]
 dyn_tassel_front_01 [{shapeType=Capsule, length=1.6197, radius=0.5138, shapeName=dyn_tassel_front_01_capsule, center=<5.7858 0 0>, rotation=<rot 90 0 0>}]
 dyn_tassel_front_01_MSTR [{shapeType=Capsule, length=0.505, radius=0.505, shapeName=dyn_tassel_front_01_MSTR_capsule, rotation=<rot 0 -90 0>}]
 dyn_tassel_front_02 [{shapeType=Capsule, length=1.6334, radius=0.5, shapeName=dyn_tassel_front_02_capsule, center=<5.7858 0 0>, rotation=<rot 90 0 0>}]
 dyn_tassel_front_02_MSTR [{shapeType=Capsule, length=0.505, radius=0.505, shapeName=dyn_tassel_front_02_MSTR_capsule, rotation=<rot 0 -90 0>}]
 dyn_tassel_front_03 [{shapeType=Capsule, length=1.6334, radius=0.5, shapeName=dyn_tassel_front_03_capsule, center=<5.7858 0 0>, rotation=<rot 90 0 0>}]
 dyn_tassel_front_03_MSTR [{shapeType=Capsule, length=0.505, radius=0.505, shapeName=dyn_tassel_front_03_MSTR_capsule, rotation=<rot 0 -90 0>}]
 dyn_tassel_front_04 [{shapeType=Capsule, length=1.6334, radius=0.5, shapeName=dyn_tassel_front_03_capsule, center=<5.7858 0 0>, rotation=<rot 90 0 0>}]
 dyn_tassel_front_04_MSTR [{shapeType=Capsule, length=0.505, radius=0.505, shapeName=dyn_tassel_front_04_MSTR_capsule, rotation=<rot 0 -90 0>}]
 neck_01 [{shapeType=Box, extentX=1.01, extentY=1.01, extentZ=1.01, shapeName=neck_01_box, rotation=<rot 0 -90 0>}]
 neck_02 [{shapeType=Box, extentX=8.9091, extentY=2.657, extentZ=9.833, shapeName=neck_02_box, center=<-1.1393 -0.8246 0>, rotation=<rot 0 -90 0>}]
 pelvis [
  {shapeType=Capsule, length=6.5268, radius=8.9681, shapeName=pelvis_capsule}
  {shapeType=Capsule, length=10.2271, radius=10.464, shapeName=pelvis_capsule, center=<-1.8572 -2.8989 0>}
  {shapeType=Capsule, length=12.2181, radius=8.9681, shapeName=pelvis_capsule, center=<0 -2.3749 0>}
  {shapeType=Capsule, length=15.0338, radius=9.9103, shapeName=pelvis_capsule, center=<-4.7243 -4.8181 0>}
 ]
 spine_01 [
  {shapeType=Capsule, length=4.2535, radius=8.9681, shapeName=spine_01_capsule}
  {shapeType=Capsule, length=0, radius=8.9681, shapeName=spine_01_capsule, center=<-0.1315 -0.7231 3.0953>}
  {shapeType=Capsule, length=0, radius=8.9681, shapeName=spine_01_capsule, center=<-0.2536 -0.7231 -3.2629>}
 ]
 spine_02 [{shapeType=Capsule, length=10.7974, radius=8.9681, shapeName=spine_02_capsule}]
 spine_04 [{shapeType=Capsule, length=13.2133, radius=11.4581, shapeName=spine_05_capsule}]
 spine_05 [{shapeType=Capsule, length=16.5767, radius=8.0947, shapeName=spine_05_capsule}]
}
constraints [
 {
  bone1Name=spine_01
  bone2Name=pelvis
  swing1LimitDegrees=45
  swing1Motion=Free
  swing2LimitDegrees=45
  swing2Motion=Free
  twistLimitDegrees=45
  twistMotion=Free
 }
 {
  bone1Name=spine_02
  bone2Name=spine_01
  swing1LimitDegrees=45
  swing1Motion=Free
  swing2LimitDegrees=45
  swing2Motion=Free
  twistLimitDegrees=45
  twistMotion=Free
 }
 {
  bone1Name=dyn_TopZipper_MSTR
  bone2Name=spine_05
  swing1LimitDegrees=45
  swing1Motion=Free
  swing2LimitDegrees=45
  swing2Motion=Free
  twistLimitDegrees=45
  twistMotion=Free
 }
 {
  bone1Name=dyn_BTMZipper_MSTR
  bone2Name=spine_02
  swing1LimitDegrees=45
  swing1Motion=Free
  swing2LimitDegrees=45
  swing2Motion=Free
  twistLimitDegrees=45
  twistMotion=Free
 }
 {bone1Name=dyn_BTMZipper, bone2Name=dyn_BTMZipper_MSTR, swing1LimitDegrees=30, swing2LimitDegrees=30, twistLimitDegrees=5}
 {
  bone1Name=spine_05
  bone2Name=spine_04
  swing1LimitDegrees=45
  swing1Motion=Free
  swing2LimitDegrees=45
  swing2Motion=Free
  twistLimitDegrees=45
  twistMotion=Free
 }
 {
  bone1Name=spine_04
  bone2Name=spine_02
  swing1LimitDegrees=45
  swing1Motion=Free
  swing2LimitDegrees=45
  swing2Motion=Free
  twistLimitDegrees=45
  twistMotion=Free
 }
 {bone1Name=dyn_TopZipper, bone2Name=dyn_TopZipper_MSTR, swing1LimitDegrees=30, swing2LimitDegrees=30, twistLimitDegrees=5}
 {
  bone1Name=dyn_tassel_front_01_MSTR
  bone2Name=spine_02
  swing1LimitDegrees=45
  swing1Motion=Free
  swing2LimitDegrees=45
  swing2Motion=Free
  twistLimitDegrees=45
  twistMotion=Free
 }
 {
  bone1Name=dyn_tassel_front_02_MSTR
  bone2Name=spine_02
  swing1LimitDegrees=45
  swing1Motion=Free
  swing2LimitDegrees=45
  swing2Motion=Free
  twistLimitDegrees=45
  twistMotion=Free
 }
 {
  bone1Name=dyn_tassel_front_03_MSTR
  bone2Name=spine_02
  swing1LimitDegrees=45
  swing1Motion=Free
  swing2LimitDegrees=45
  swing2Motion=Free
  twistLimitDegrees=45
  twistMotion=Free
 }
 {
  bone1Name=dyn_tassel_front_04_MSTR
  bone2Name=spine_02
  swing1LimitDegrees=45
  swing1Motion=Free
  swing2LimitDegrees=45
  swing2Motion=Free
  twistLimitDegrees=45
  twistMotion=Free
 }
 {bone1Name=dyn_tassel_front_01, bone2Name=dyn_tassel_front_01_MSTR, swing1LimitDegrees=20, swing2LimitDegrees=20, twistLimitDegrees=5}
 {bone1Name=dyn_tassel_front_02, bone2Name=dyn_tassel_front_02_MSTR, swing1LimitDegrees=20, swing2LimitDegrees=20, twistLimitDegrees=5}
 {bone1Name=dyn_tassel_front_03, bone2Name=dyn_tassel_front_03_MSTR, swing1LimitDegrees=20, swing2LimitDegrees=20, twistLimitDegrees=5}
 {bone1Name=dyn_tassel_front_04, bone2Name=dyn_tassel_front_04_MSTR, swing1LimitDegrees=20, swing2LimitDegrees=20, twistLimitDegrees=5}
 {
  bone1Name=dyn_necklaceMain
  bone2Name=dyn_necklaceMain_Base
  swing1LimitDegrees=45
  swing1Motion=Free
  swing2LimitDegrees=45
  swing2Motion=Free
  twistLimitDegrees=45
  twistMotion=Free
 }
 {bone1Name=dyn_necklace_01, bone2Name=dyn_necklaceMain, swing1LimitDegrees=20, swing2LimitDegrees=20, twistLimitDegrees=5}
 {bone1Name=dyn_necklace_02, bone2Name=dyn_necklace_01, swing1LimitDegrees=20, swing2LimitDegrees=20, twistLimitDegrees=5}
 {
  bone1Name=dyn_necklaceMain_Base
  bone2Name=spine_05
  swing1LimitDegrees=45
  swing1Motion=Free
  swing2LimitDegrees=45
  swing2Motion=Free
  twistLimitDegrees=45
  twistMotion=Free
 }
 {
  bone1Name=neck_02
  bone2Name=neck_01
  swing1LimitDegrees=45
  swing1Motion=Free
  swing2LimitDegrees=45
  swing2Motion=Free
  twistLimitDegrees=45
  twistMotion=Free
 }
 {
  bone1Name=neck_01
  bone2Name=spine_05
  swing1LimitDegrees=45
  swing1Motion=Free
  swing2LimitDegrees=45
  swing2Motion=Free
  twistLimitDegrees=45
  twistMotion=Free
 }
 {bone1Name=dyn_tassel_back_01_a_MSTR, bone2Name=pelvis, swing1LimitDegrees=0, swing2LimitDegrees=0, twistLimitDegrees=0}
 {bone1Name=dyn_tassel_back_01_b_MSTR, bone2Name=pelvis, swing1LimitDegrees=0, swing2LimitDegrees=0, twistLimitDegrees=0}
 {bone1Name=dyn_tassel_back_01_a, bone2Name=dyn_tassel_back_01_a_MSTR, swing1LimitDegrees=45, swing2LimitDegrees=45, twistLimitDegrees=0}
 {bone1Name=dyn_tassel_back_02_a, bone2Name=dyn_tassel_back_01_a, swing1LimitDegrees=40, swing2LimitDegrees=40, twistLimitDegrees=0}
 {bone1Name=dyn_tassel_back_01_02_b, bone2Name=dyn_tassel_back_01_01_b, swing1LimitDegrees=40, swing2LimitDegrees=40, twistLimitDegrees=0}
 {bone1Name=dyn_tassel_back_01_01_b, bone2Name=dyn_tassel_back_01_b_MSTR, swing1LimitDegrees=45, swing2LimitDegrees=45, twistLimitDegrees=0}
]
