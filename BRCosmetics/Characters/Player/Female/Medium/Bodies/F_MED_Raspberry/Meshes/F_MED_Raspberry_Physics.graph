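# Physics asset for the F_MED_Raspberry female player body, set up for a dangling strap on the right leg.
# A four-capsule strap chain hangs from dyn_legStrap_MSTR with limited swing and twist between links.
# Both strap master bodies, including the childless dyn_legStrapTop_MSTR, link to thigh_r through free angular constraints.
# Capsules cover the right thigh and calf, and a wide, thin root box sits at ground level.
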PhysicsAsset /BRCosmetics/Characters/Player/Female/Medium/Bodies/F_MED_Raspberry/Meshes/F_MED_Raspberry_Physics
bodies {
 calf_r [{shapeType=Capsule, length=34.4439, radius=8.4269, shapeName=calf_r_capsule, center=<21.6503 0 0>, rotation=<rot 90 0 0>}]
 dyn_legStrapTop_MSTR [{shapeType=Capsule, length=0.505, radius=0.505, shapeName=dyn_legStrapTop_MSTR_capsule, rotation=<rot 0 -90 0>}]
 dyn_legStrap_01 [{shapeType=Capsule, length=6.2737, radius=2.2456, shapeName=dyn_legStrap_01_capsule, center=<4.5917 0 4>, rotation=<rot 90 0 0>}]
 dyn_legStrap_02 [{shapeType=Capsule, length=5.53, radius=2.9893, shapeName=dyn_legStrap_02_capsule, center=<4.5917 0 4>, rotation=<rot 90 0 0>}]
 dyn_legStrap_03 [{shapeType=Capsule, length=5.1986, radius=3.3207, shapeName=dyn_legStrap_03_capsule, center=<4.5917 0 4>, rotation=<rot 90 0 0>}]
 dyn_legStrap_04 [{shapeType=Capsule, length=5.1986, radius=3.3207, shapeName=dyn_legStrap_04_capsule, center=<4.5917 0 4>, rotation=<rot 90 0 0>}]
 dyn_legStrap_MSTR [{shapeType=Capsule, length=0.505, radius=0.505, shapeName=dyn_legStrap_MSTR_capsule, rotation=<rot 0 -90 0>}]
 root [{shapeType=Box, extentX=229.2545, extentY=186.4503, extentZ=11.15, shapeName=root_box, center=<0 0 -3.7076>, rotation=<rot 0 -90 0>}]
 thigh_r [{shapeType=Capsule, length=29.1708, radius=8.3076, shapeName=thigh_r_capsule, center=<22.6182 0 0>, rotation=<rot 90 0 0>}]
}
constraints [
 {bone1Name=dyn_legStrap_01, bone2Name=dyn_legStrap_MSTR, swing1LimitDegrees=45, swing2LimitDegrees=45, twistLimitDegrees=30}
 {bone1Name=dyn_legStrap_02, bone2Name=dyn_legStrap_01, swing1LimitDegrees=45, swing2LimitDegrees=45, twistLimitDegrees=30}
 {bone1Name=dyn_legStrap_03, bone2Name=dyn_legStrap_02, swing1LimitDegrees=45, swing2LimitDegrees=45, twistLimitDegrees=30}
 {bone1Name=dyn_legStrap_04, bone2Name=dyn_legStrap_03, swing1LimitDegrees=45, swing2LimitDegrees=45, twistLimitDegrees=30}
 {
  bone1Name=dyn_legStrapTop_MSTR
  bone2Name=thigh_r
  swing1LimitDegrees=45
  swing1Motion=Free
  swing2LimitDegrees=45
  swing2Motion=Free
  twistLimitDegrees=45
  twistMotion=Free
 }
 {
  bone1Name=dyn_legStrap_MSTR
  bone2Name=thigh_r
  swing1LimitDegrees=45
  swing1Motion=Free
  swing2LimitDegrees=45
  swing2Motion=Free
  twistLimitDegrees=45
  twistMotion=Free
 }
 {
  bone1Name=calf_r
  bone2Name=thigh_r
  swing1LimitDegrees=45
  swing1Motion=Free
  swing2LimitDegrees=45
  swing2Motion=Free
  twistLimitDegrees=45
  twistMotion=Free
 }
 {
  bone1Name=thigh_r
  bone2Name=root
  swing1LimitDegrees=45
  swing1Motion=Free
  swing2LimitDegrees=45
  swing2Motion=Free
  twistLimitDegrees=45
  twistMotion=Free
 }
]
