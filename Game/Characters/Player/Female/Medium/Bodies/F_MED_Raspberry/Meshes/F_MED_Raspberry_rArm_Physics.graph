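# Physics asset for the right arm of the medium female Raspberry player body, rigging an arm strap for physics-driven motion.
# Capsules cover the upper arm, its two twist bones, a small strap master bone and three chained strap segments.
# The strap master hangs off the first twist bone, and both twist bones join the upper arm, all with free rotation.
# Constraints along the strap segments carry swing and twist limits.
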PhysicsAsset /Game/Characters/Player/Female/Medium/Bodies/F_MED_Raspberry/Meshes/F_MED_Raspberry_rArm_Physics
bodies {
 dyn_r_armStrap_01 [{shapeType=Capsule, length=2.6885, radius=1.3115, shapeName=dyn_r_armStrap_01_capsule, center=<2.5163 0 0>, rotation=<rot 90 0 0>}]
 dyn_r_armStrap_02 [{shapeType=Capsule, length=2.629, radius=1.371, shapeName=dyn_r_armStrap_02_capsule, center=<2.5163 0 0>, rotation=<rot 90 0 0>}]
 dyn_r_armStrap_03 [{shapeType=Capsule, length=2.2309, radius=1.7691, shapeName=dyn_r_armStrap_03_capsule, center=<2.5163 0 0>, rotation=<rot 90 0 0>}]
 dyn_r_armStrap_MSTR [{shapeType=Capsule, length=0.505, radius=0.505, shapeName=dyn_r_armStrap_MSTR_capsule, rotation=<rot 0 -90 0>}]
 upperarm_r [{shapeType=Capsule, length=29.8764, radius=6.0075, shapeName=upperarm_r_capsule, center=<-14.4585 0 0>, rotation=<rot 90 0 0>}]
 upperarm_twist_01_r [{shapeType=Capsule, length=20.4133, radius=5.9931, shapeName=upperarm_twist_01_r_capsule, rotation=<rot 90 0 0>}]
 upperarm_twist_02_r [{shapeType=Capsule, length=9.7827, radius=0.5, shapeName=upperarm_twist_02_r_capsule, rotation=<rot 90 0 0>}]
}
constraints [
 {
  bone1Name=dyn_r_armStrap_MSTR
  bone2Name=upperarm_twist_01_r
  swing1LimitDegrees=45
  swing1Motion=Free
  swing2LimitDegrees=45
  swing2Motion=Free
  twistLimitDegrees=45
  twistMotion=Free
 }
 {bone1Name=dyn_r_armStrap_01, bone2Name=dyn_r_armStrap_MSTR, swing1LimitDegrees=60, swing2LimitDegrees=60, twistLimitDegrees=45}
 {bone1Name=dyn_r_armStrap_02, bone2Name=dyn_r_armStrap_01, swing1LimitDegrees=60, swing2LimitDegrees=60, twistLimitDegrees=45}
 {bone1Name=dyn_r_armStrap_03, bone2Name=dyn_r_armStrap_02, swing1LimitDegrees=60, swing2LimitDegrees=60, twistLimitDegrees=45}
 {
  bone1Name=upperarm_twist_02_r
  bone2Name=upperarm_r
  swing1LimitDegrees=45
  swing1Motion=Free
  swing2LimitDegrees=45
  swing2Motion=Free
  twistLimitDegrees=45
  twistMotion=Free
 }
 {
  bone1Name=upperarm_twist_01_r
  bone2Name=upperarm_r
  swing1LimitDegrees=45
  swing1Motion=Free
  swing2LimitDegrees=45
  swing2Motion=Free
  twistLimitDegrees=45
  twistMotion=Free
 }
]
