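# Post-process style AnimBlueprint for the F_MED_Raspberry female body cosmetic, adding deformation and secondary motion to straps, zippers, necklace and tassels.
# The copied parent pose runs through a deform Control Rig and weighted bone constraints into cached PreSim, then AnimDynamics, two RigidBody nodes and glute corrective offsets.
# The EventGraph computes gravity override, thigh-rotation alphas and crouch state.
# FrontEndIdle is ANDed with false, and the skydive-switched RigidBody branch never reaches output.
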AnimBlueprint /BRCosmetics/Characters/Player/Female/Medium/Bodies/F_MED_Raspberry/Meshes/F_MED_Raspberry_AnimBP
target_skeleton=&/F_MED_Raspberry_Skeleton
graphs [
 AnimGraph {nodes [
   "Constraint - dyn_tassel_back_01_a_MSTR (PPPPP)"!(
    ComponentPose
    Alpha=1
    ConstraintWeights_0=0.0076
    ConstraintWeights_1:float=0.0106
    ConstraintWeights_2:float=0.1435
    ConstraintWeights_3:float=0.0684
    ConstraintWeights_4:float=0.7699
   ) -> (Pose) {
    alphaBoolBlend {blendOption=Linear}
    boneToModify {boneName=dyn_tassel_back_01_a_MSTR}
    constraintSetup [
     {targetBone {boneName=spine_04}}
     {targetBone {boneName=spine_03}}
     {targetBone {boneName=spine_02}}
     {targetBone {boneName=spine_01}}
     {targetBone {boneName=pelvis}}
    ]
    constraintWeights [1,1,1,1,1]
   }
   "Constraint - dyn_tassel_back_01_b_MSTR (PPPPPPP)"!(
    ComponentPose
    Alpha=1
    ConstraintWeights_0=0.007
    ConstraintWeights_1:float=0.0101
    ConstraintWeights_2:float=0.1399
    ConstraintWeights_3:float=0.067
    ConstraintWeights_4:float=0.7759
    ConstraintWeights_5:float=0
    ConstraintWeights_6:float=0
   ) -> (Pose) {
    alphaBoolBlend {blendOption=Linear}
    boneToModify {boneName=dyn_tassel_back_01_b_MSTR}
    constraintSetup [
     {targetBone {boneName=spine_04}}
     {targetBone {boneName=spine_03}}
     {targetBone {boneName=spine_02}}
     {targetBone {boneName=spine_01}}
     {targetBone {boneName=pelvis}}
     {targetBone {boneName=deform_glute_l}}
     {targetBone {boneName=deform_glute_r}}
    ]
    constraintWeights [1,1,1,1,1,1,1]
   }
   n1 = RigidBody!(ComponentPose, Alpha, OverrideWorldGravity:Vector="0,0,-980", ExternalForce) -> (Pose) {
    bClampLinearTranslationLimitToRefPose=true
    bOverrideWorldGravity=true
    overridePhysicsAsset=&/F_MED_Raspberry_Physics
    simulationSpace=BaseBoneSpace
    alphaBoolBlend {blendOption=Linear}
    baseBoneRef {boneName=pelvis}
    componentAppliedLinearAccClamp=<2500 2500 2500>
    componentLinearAccScale=<0.3 0.3 0.4>
    componentLinearVelScale=<0.3 0.3 0.4>
    overrideWorldGravity {z=-980}
    simSpaceSettings {
     maxAngularAcceleration=1000
     maxAngularVelocity=2
     maxLinearAcceleration=700
     maxLinearVelocity=500
     velocityScaleZ=0
     externalLinearDragV=<0.2 0.2 0.2>
    }
   }
   n2 = RigidBody(ComponentPose=@n1, OverrideWorldGravity="0,0,-980") {
    bClampLinearTranslationLimitToRefPose=true
    bOverrideWorldGravity=true
    overridePhysicsAsset=/Game/Characters/Player/Female/Medium/Bodies/F_MED_Raspberry/Meshes/F_MED_Raspberry_lArm_Physics
    simulationSpace=BaseBoneSpace
    alphaBoolBlend {blendOption=Linear}
    baseBoneRef {boneName=upperarm_twist_01_l}
    componentAppliedLinearAccClamp=<2500 2500 2500>
    componentLinearAccScale=<0.3 0.3 0.4>
    componentLinearVelScale=<0.3 0.3 0.4>
    overrideWorldGravity {z=-980}
    simSpaceSettings {
     maxAngularAcceleration=1000
     maxAngularVelocity=2
     maxLinearAcceleration=700
     maxLinearVelocity=500
     velocityScaleZ=0
     worldAlpha=1
     externalLinearDragV=<0.2 0.2 0.2>
    }
   }
   n3 = RigidBody(ComponentPose=@n2, OverrideWorldGravity="0,0,-980") {
    bClampLinearTranslationLimitToRefPose=true
    bOverrideWorldGravity=true
    overridePhysicsAsset=/Game/Characters/Player/Female/Medium/Bodies/F_MED_Raspberry/Meshes/F_MED_Raspberry_rArm_Physics
    simulationSpace=BaseBoneSpace
    alphaBoolBlend {blendOption=Linear}
    baseBoneRef {boneName=upperarm_twist_01_r}
    componentAppliedLinearAccClamp=<2500 2500 2500>
    componentLinearAccScale=<0.3 0.3 0.4>
    componentLinearVelScale=<0.3 0.3 0.4>
    overrideWorldGravity {z=-980}
    simSpaceSettings {
     maxAngularAcceleration=1000
     maxAngularVelocity=2
     maxLinearAcceleration=700
     maxLinearVelocity=500
     velocityScaleZ=0
     worldAlpha=1
     externalLinearDragV=<0.2 0.2 0.2>
    }
   }
   n4 = RigidBody(ComponentPose=Local To Component(LocalPose=Use cached pose 'PreSim'()), OverrideWorldGravity="0,0,350") {
    bClampLinearTranslationLimitToRefPose=true
    bOverrideWorldGravity=true
    overridePhysicsAsset=&F_MED_Raspberry_Physics
    simulationSpace=BaseBoneSpace
    alphaBoolBlend {blendOption=Linear}
    baseBoneRef {boneName=pelvis}
    componentAppliedLinearAccClamp=<2500 2500 2500>
    componentLinearAccScale=<0.3 0.3 0.4>
    componentLinearVelScale=<0.3 0.3 0.4>
    overrideWorldGravity {z=-980}
    simSpaceSettings {
     maxAngularAcceleration=1000
     maxAngularVelocity=2
     maxLinearAcceleration=700
     maxLinearVelocity=500
     velocityScaleZ=0
     externalLinearDragV=<0.2 0.2 0.2>
    }
   }
   n5 = RigidBody(ComponentPose=@n4, OverrideWorldGravity="0,0,350") {
    bClampLinearTranslationLimitToRefPose=true
    bOverrideWorldGravity=true
    overridePhysicsAsset=&F_MED_Raspberry_lArm_Physics
    simulationSpace=BaseBoneSpace
    alphaBoolBlend {blendOption=Linear}
    baseBoneRef {boneName=upperarm_twist_01_l}
    componentAppliedLinearAccClamp=<2500 2500 2500>
    componentLinearAccScale=<0.3 0.3 0.4>
    componentLinearVelScale=<0.3 0.3 0.4>
    overrideWorldGravity {z=-980}
    simSpaceSettings {
     maxAngularAcceleration=1000
     maxAngularVelocity=2
     maxLinearAcceleration=700
     maxLinearVelocity=500
     velocityScaleZ=0
     worldAlpha=1
     externalLinearDragV=<0.2 0.2 0.2>
    }
   }
   n6 = RigidBody(ComponentPose=@n5, OverrideWorldGravity="0,0,350") {
    bClampLinearTranslationLimitToRefPose=true
    bOverrideWorldGravity=true
    overridePhysicsAsset=&F_MED_Raspberry_rArm_Physics
    simulationSpace=BaseBoneSpace
    alphaBoolBlend {blendOption=Linear}
    baseBoneRef {boneName=upperarm_twist_01_r}
    componentAppliedLinearAccClamp=<2500 2500 2500>
    componentLinearAccScale=<0.3 0.3 0.4>
    componentLinearVelScale=<0.3 0.3 0.4>
    overrideWorldGravity {z=-980}
    simSpaceSettings {
     maxAngularAcceleration=1000
     maxAngularVelocity=2
     maxLinearAcceleration=700
     maxLinearVelocity=500
     velocityScaleZ=0
     worldAlpha=1
     externalLinearDragV=<0.2 0.2 0.2>
    }
   }
   n7 = Get bIsSkyDiving()
   n8 = Get bIsInFrontEnd()
   Local To Component(LocalPose=Blend Poses by bool(BlendPose_0=Component To Local(ComponentPose=@n6), BlendPose_1=Component To Local(ComponentPose=@n3), bActiveValue=AND Boolean(A=@n7, B=@n8)) {blendType=Linear})
   n9 = Anim Dynamics - Bone: dyn_legStrap_01!(ComponentPose=Local To Component(LocalPose=Use cached pose 'PreSim'()), Alpha=1, GravityOverride:Vector="0,0,-980", ExternalForce="0,0,0") -> (Pose) {
    angularDampingOverride=1
    angularSpringConstant=100
    bOverrideAngularDamping=true
    bUseGravityOverride=true
    simulationSpace=BoneRelative
    alphaBoolBlend {blendOption=Linear}
    boundBone {boneName=dyn_legStrap_01}
    componentAppliedLinearAccClamp=<2000 2000 7000>
    componentLinearAccScale=<0.5 0.5 0.5>
    componentLinearVelScale=<0.5 0.5 0.5>
    gravityOverride {z=-980}
    physicsBodyDefinitions [{
      sphereCollisionRadius=0
      boundBone {boneName=dyn_legStrap_01}
      boxExtents=<11 18 18>
      constraintSetup {angularLimitsMax=<45 20 45>, angularLimitsMin=<-45 -45 -45>, angularTarget=<0 0 0>}
      localJointOffset=<6 0 0>
     }]
    relativeSpaceBone {boneName=root}
   }
   n10 = Anim Dynamics - Bone: dyn_legStrap_02!(ComponentPose=@n9, Alpha=1, GravityOverride:Vector="0,0,-980", ExternalForce="0,0,0") -> (Pose) {
    angularDampingOverride=1
    angularSpringConstant=100
    bOverrideAngularDamping=true
    bUseGravityOverride=true
    simulationSpace=BoneRelative
    alphaBoolBlend {blendOption=Linear}
    boundBone {boneName=dyn_legStrap_02}
    componentAppliedLinearAccClamp=<2000 2000 7000>
    componentLinearAccScale=<0.5 0.5 0.5>
    componentLinearVelScale=<0.5 0.5 0.5>
    gravityOverride {z=-980}
    physicsBodyDefinitions [{
      sphereCollisionRadius=0
      boundBone {boneName=dyn_legStrap_02}
      boxExtents=<11 18 18>
      constraintSetup {angularLimitsMax=<45 20 45>, angularLimitsMin=<-45 -45 -45>, angularTarget=<0 0 0>}
      localJointOffset=<6 0 0>
     }]
    relativeSpaceBone {boneName=root}
   }
   n11 = Anim Dynamics - Bone: dyn_legStrap_03!(ComponentPose=@n10, Alpha=1, GravityOverride:Vector="0,0,-980", ExternalForce="0,0,0") -> (Pose) {
    angularDampingOverride=1
    angularSpringConstant=100
    bOverrideAngularDamping=true
    bUseGravityOverride=true
    simulationSpace=BoneRelative
    alphaBoolBlend {blendOption=Linear}
    boundBone {boneName=dyn_legStrap_03}
    componentAppliedLinearAccClamp=<2000 2000 7000>
    componentLinearAccScale=<0.5 0.5 0.5>
    componentLinearVelScale=<0.5 0.5 0.5>
    gravityOverride {z=-980}
    physicsBodyDefinitions [{
      sphereCollisionRadius=0
      boundBone {boneName=dyn_legStrap_03}
      boxExtents=<11 18 18>
      constraintSetup {angularLimitsMax=<45 20 45>, angularLimitsMin=<-45 -45 -45>, angularTarget=<0 0 0>}
      localJointOffset=<6 0 0>
     }]
    relativeSpaceBone {boneName=root}
   }
   n12 = Anim Dynamics - Bone: dyn_legStrap_04!(ComponentPose=@n11, Alpha=1, GravityOverride:Vector="0,0,-980", ExternalForce="0,0,0") -> (Pose) {
    angularDampingOverride=1
    angularSpringConstant=100
    bOverrideAngularDamping=true
    bUseGravityOverride=true
    simulationSpace=BoneRelative
    alphaBoolBlend {blendOption=Linear}
    boundBone {boneName=dyn_legStrap_04}
    componentAppliedLinearAccClamp=<2000 2000 7000>
    componentLinearAccScale=<0.5 0.5 0.5>
    componentLinearVelScale=<0.5 0.5 0.5>
    gravityOverride {z=-980}
    physicsBodyDefinitions [{
      sphereCollisionRadius=0
      boundBone {boneName=dyn_legStrap_04}
      boxExtents=<11 18 18>
      constraintSetup {angularLimitsMax=<45 20 45>, angularLimitsMin=<-45 -45 -45>, angularTarget=<0 0 0>}
      localJointOffset=<6 0 0>
     }]
    relativeSpaceBone {boneName=root}
   }
   n13 = Anim Dynamics - Bone: dyn_l_armStrap_01!(ComponentPose=@n12, Alpha=1, GravityOverride:Vector="0,0,0", ExternalForce="0,0,0") -> (Pose) {
    angularDampingOverride=0.999
    bOverrideAngularDamping=true
    bUseGravityOverride=true
    bUseSphericalLimits=true
    simulationSpace=RootRelative
    alphaBoolBlend {blendOption=Linear}
    boundBone {boneName=dyn_l_armStrap_01}
    componentAppliedLinearAccClamp=<5000 5000 5000>
    componentLinearAccScale=<0.5 0.5 0.5>
    componentLinearVelScale=<0.5 0.5 0.5>
    physicsBodyDefinitions [{
      sphereCollisionRadius=0
      boundBone {boneName=dyn_l_armStrap_01}
      boxExtents=<12 12 5>
      constraintSetup {angularLimitsMax=<60 60 60>, angularLimitsMin=<-60 -60 -60>, angularTarget=<0 0 0>}
      localJointOffset=<2 0 0>
     }]
    relativeSpaceBone {boneName=upperarm_twist_01_l}
    sphericalLimits [
     {limitRadius=8, drivingBone {boneName=upperarm_twist_02_l}, sphereLocalOffset=<-5 0 0>}
     {limitRadius=8, drivingBone {boneName=upperarm_l}, sphereLocalOffset=<5 0 -2>}
    ]
   }
   n14 = Anim Dynamics - Bone: dyn_l_armStrap_02!(ComponentPose=@n13, Alpha=1, GravityOverride:Vector="0,0,0", ExternalForce="0,0,0") -> (Pose) {
    angularDampingOverride=0.999
    bOverrideAngularDamping=true
    bUseGravityOverride=true
    bUseSphericalLimits=true
    simulationSpace=RootRelative
    alphaBoolBlend {blendOption=Linear}
    boundBone {boneName=dyn_l_armStrap_02}
    componentAppliedLinearAccClamp=<5000 5000 5000>
    componentLinearAccScale=<0.5 0.5 0.5>
    componentLinearVelScale=<0.5 0.5 0.5>
    physicsBodyDefinitions [{
      sphereCollisionRadius=0
      boundBone {boneName=dyn_l_armStrap_02}
      boxExtents=<12 12 5>
      constraintSetup {angularLimitsMax=<60 60 60>, angularLimitsMin=<-60 -60 -60>, angularTarget=<0 0 0>}
      localJointOffset=<2 0 0>
     }]
    relativeSpaceBone {boneName=upperarm_twist_01_l}
    sphericalLimits [
     {limitRadius=8, drivingBone {boneName=upperarm_twist_02_l}, sphereLocalOffset=<-5 0 0>}
     {limitRadius=8, drivingBone {boneName=upperarm_l}, sphereLocalOffset=<5 0 -2>}
    ]
   }
   n15 = Anim Dynamics - Bone: dyn_l_armStrap_03!(ComponentPose=@n14, Alpha=1, GravityOverride:Vector="0,0,0", ExternalForce="0,0,0") -> (Pose) {
    angularDampingOverride=0.999
    bOverrideAngularDamping=true
    bUseGravityOverride=true
    bUseSphericalLimits=true
    simulationSpace=RootRelative
    alphaBoolBlend {blendOption=Linear}
    boundBone {boneName=dyn_l_armStrap_03}
    componentAppliedLinearAccClamp=<5000 5000 5000>
    componentLinearAccScale=<0.5 0.5 0.5>
    componentLinearVelScale=<0.5 0.5 0.5>
    physicsBodyDefinitions [{
      sphereCollisionRadius=0
      boundBone {boneName=dyn_l_armStrap_03}
      boxExtents=<12 12 5>
      constraintSetup {angularLimitsMax=<60 60 60>, angularLimitsMin=<-60 -60 -60>, angularTarget=<0 0 0>}
      localJointOffset=<2 0 0>
     }]
    relativeSpaceBone {boneName=upperarm_twist_01_l}
    sphericalLimits [
     {limitRadius=8, drivingBone {boneName=upperarm_twist_02_l}, sphereLocalOffset=<-5 0 0>}
     {limitRadius=8, drivingBone {boneName=upperarm_l}, sphereLocalOffset=<5 0 -2>}
    ]
   }
   n16 = Anim Dynamics - Bone: dyn_r_armStrap_01!(ComponentPose=@n15, Alpha=1, GravityOverride:Vector="0,0,0", ExternalForce="0,0,0") -> (Pose) {
    angularDampingOverride=0.999
    bOverrideAngularDamping=true
    bUseGravityOverride=true
    bUseSphericalLimits=true
    simulationSpace=RootRelative
    alphaBoolBlend {blendOption=Linear}
    boundBone {boneName=dyn_r_armStrap_01}
    componentAppliedLinearAccClamp=<5000 5000 5000>
    componentLinearAccScale=<0.5 0.5 0.5>
    componentLinearVelScale=<0.5 0.5 0.5>
    physicsBodyDefinitions [{
      sphereCollisionRadius=0
      boundBone {boneName=dyn_r_armStrap_01}
      boxExtents=<12 12 5>
      constraintSetup {angularLimitsMax=<60 60 60>, angularLimitsMin=<-60 -60 -60>, angularTarget=<0 0 0>}
      localJointOffset=<2 0 0>
     }]
    relativeSpaceBone {boneName=upperarm_twist_01_r}
    sphericalLimits [
     {limitRadius=8, drivingBone {boneName=upperarm_twist_02_r}, sphereLocalOffset=<5 0 0>}
     {limitRadius=8, drivingBone {boneName=upperarm_r}, sphereLocalOffset=<-5 0 2>}
    ]
   }
   n17 = Anim Dynamics - Bone: dyn_r_armStrap_02!(ComponentPose=@n16, Alpha=1, GravityOverride:Vector="0,0,0", ExternalForce="0,0,0") -> (Pose) {
    angularDampingOverride=0.999
    bOverrideAngularDamping=true
    bUseGravityOverride=true
    bUseSphericalLimits=true
    simulationSpace=RootRelative
    alphaBoolBlend {blendOption=Linear}
    boundBone {boneName=dyn_r_armStrap_02}
    componentAppliedLinearAccClamp=<5000 5000 5000>
    componentLinearAccScale=<0.5 0.5 0.5>
    componentLinearVelScale=<0.5 0.5 0.5>
    physicsBodyDefinitions [{
      sphereCollisionRadius=0
      boundBone {boneName=dyn_r_armStrap_02}
      boxExtents=<12 12 5>
      constraintSetup {angularLimitsMax=<60 60 60>, angularLimitsMin=<-60 -60 -60>, angularTarget=<0 0 0>}
      localJointOffset=<2 0 0>
     }]
    relativeSpaceBone {boneName=upperarm_twist_01_r}
    sphericalLimits [
     {limitRadius=8, drivingBone {boneName=upperarm_twist_02_r}, sphereLocalOffset=<5 0 0>}
     {limitRadius=8, drivingBone {boneName=upperarm_r}, sphereLocalOffset=<-5 0 2>}
    ]
   }
   n18 = Anim Dynamics - Bone: dyn_r_armStrap_03!(ComponentPose=@n17, Alpha=1, GravityOverride:Vector="0,0,0", ExternalForce="0,0,0") -> (Pose) {
    angularDampingOverride=0.999
    bOverrideAngularDamping=true
    bUseGravityOverride=true
    bUseSphericalLimits=true
    simulationSpace=RootRelative
    alphaBoolBlend {blendOption=Linear}
    boundBone {boneName=dyn_r_armStrap_03}
    componentAppliedLinearAccClamp=<5000 5000 5000>
    componentLinearAccScale=<0.5 0.5 0.5>
    componentLinearVelScale=<0.5 0.5 0.5>
    physicsBodyDefinitions [{
      sphereCollisionRadius=0
      boundBone {boneName=dyn_r_armStrap_03}
      boxExtents=<12 12 5>
      constraintSetup {angularLimitsMax=<60 60 60>, angularLimitsMin=<-60 -60 -60>, angularTarget=<0 0 0>}
      localJointOffset=<2 0 0>
     }]
    relativeSpaceBone {boneName=upperarm_twist_01_r}
    sphericalLimits [
     {limitRadius=8, drivingBone {boneName=upperarm_twist_02_r}, sphereLocalOffset=<5 0 0>}
     {limitRadius=8, drivingBone {boneName=upperarm_r}, sphereLocalOffset=<-5 0 2>}
    ]
   }
   n19 = Anim Dynamics - Bone: dyn_TopZipper(ComponentPose=@n18, Alpha=1, ExternalForce="0,0,0") {
    angularDampingOverride=0.9999
    bGravityOverrideInSimSpace=true
    bOverrideAngularDamping=true
    simulationSpace=RootRelative
    alphaBoolBlend {blendOption=Linear}
    boundBone {boneName=dyn_TopZipper}
    physicsBodyDefinitions [{
      sphereCollisionRadius=0
      boundBone {boneName=dyn_TopZipper}
      boxExtents=<3 10 10>
      constraintSetup {angularLimitsMax=<2 5 45>, angularLimitsMin=<-2 -45 -45>, angularTarget=<0 0 0>}
      localJointOffset=<1.5 -0.05 0.1>
     }]
   }
   n20 = Anim Dynamics - Bone: dyn_BTMZipper(ComponentPose=@n19, Alpha=1, ExternalForce="0,0,0") {
    angularDampingOverride=0.9999
    bGravityOverrideInSimSpace=true
    bOverrideAngularDamping=true
    simulationSpace=RootRelative
    alphaBoolBlend {blendOption=Linear}
    boundBone {boneName=dyn_BTMZipper}
    physicsBodyDefinitions [{
      sphereCollisionRadius=0
      boundBone {boneName=dyn_BTMZipper}
      boxExtents=<3 10 10>
      constraintSetup {angularLimitsMax=<2 5 45>, angularLimitsMin=<-2 -45 -45>, angularTarget=<0 0 0>}
      localJointOffset=<1.5 -0.05 0.1>
     }]
   }
   n21 = Anim Dynamics - Bone: dyn_necklaceMain(ComponentPose=@n20, Alpha=1, ExternalForce="0,0,0") {
    angularDampingOverride=1
    angularSpringConstant=1500
    bAngularSpring=true
    bGravityOverrideInSimSpace=true
    bOverrideAngularDamping=true
    gravityScale=0.35
    simulationSpace=RootRelative
    alphaBoolBlend {blendOption=Linear}
    alphaScaleBiasClamp {scale=0.7}
    boundBone {boneName=dyn_necklaceMain}
    componentLinearAccScale=<0.4 0.4 0.46>
    componentLinearVelScale=<0.4 0.4 0.46>
    physicsBodyDefinitions [{
      sphereCollisionRadius=0
      boundBone {boneName=dyn_necklaceMain}
      boxExtents=<10 20 20>
      constraintSetup {angularLimitsMax=<2 2 -2>, angularLimitsMin=<-2 -2 -20>, angularTarget=<0 0 0>}
      localJointOffset=<5 0 0>
     }]
   }
   n22 = Anim Dynamics - Bone: dyn_necklace_01(ComponentPose=@n21, Alpha=1, ExternalForce="0,0,0") {
    angularDampingOverride=1
    angularSpringConstant=1500
    bAngularSpring=true
    bGravityOverrideInSimSpace=true
    bOverrideAngularDamping=true
    gravityScale=0.35
    simulationSpace=RootRelative
    alphaBoolBlend {blendOption=Linear}
    alphaScaleBiasClamp {scale=0.7}
    boundBone {boneName=dyn_necklace_01}
    componentLinearAccScale=<0.4 0.4 0.46>
    componentLinearVelScale=<0.4 0.4 0.46>
    physicsBodyDefinitions [{
      sphereCollisionRadius=0
      boundBone {boneName=dyn_necklace_01}
      boxExtents=<10 20 20>
      constraintSetup {angularLimitsMax=<2 2 -2>, angularLimitsMin=<-2 -2 -20>, angularTarget=<0 0 0>}
      localJointOffset=<5 0 0>
     }]
   }
   n23 = Anim Dynamics - Bone: dyn_necklace_02(ComponentPose=@n22, Alpha=1, ExternalForce="0,0,0") {
    angularDampingOverride=1
    angularSpringConstant=1500
    bAngularSpring=true
    bGravityOverrideInSimSpace=true
    bOverrideAngularDamping=true
    gravityScale=0.35
    simulationSpace=RootRelative
    alphaBoolBlend {blendOption=Linear}
    alphaScaleBiasClamp {scale=0.7}
    boundBone {boneName=dyn_necklace_02}
    componentLinearAccScale=<0.4 0.4 0.46>
    componentLinearVelScale=<0.4 0.4 0.46>
    physicsBodyDefinitions [{
      sphereCollisionRadius=0
      boundBone {boneName=dyn_necklace_02}
      boxExtents=<10 20 20>
      constraintSetup {angularLimitsMax=<2 2 -2>, angularLimitsMin=<-2 -2 -20>, angularTarget=<0 0 0>}
      localJointOffset=<5 0 0>
     }]
   }
   n24 = Anim Dynamics - Bone: dyn_tassel_front_01!(ComponentPose=@n23, Alpha=0.7, GravityOverride:Vector="0,0,0", ExternalForce="0,0,0") -> (Pose) {
    angularDampingOverride=1
    bGravityOverrideInSimSpace=true
    bOverrideAngularDamping=true
    bUseGravityOverride=true
    bUseSphericalLimits=true
    simulationSpace=RootRelative
    alphaBoolBlend {blendOption=Linear}
    boundBone {boneName=dyn_tassel_front_01}
    componentAppliedLinearAccClamp=<5000 5000 5000>
    componentLinearAccScale=<0.5 0.5 0.5>
    componentLinearVelScale=<0.5 0.5 0.5>
    physicsBodyDefinitions [{
      sphereCollisionRadius=0
      boundBone {boneName=dyn_tassel_front_01}
      boxExtents=<20 20 10>
      constraintSetup {angularLimitsMax=<25 -12 25>, angularLimitsMin=<-25 -25 -25>, angularTarget=<0 0 0>}
      localJointOffset=<5 0 0>
     }]
    sphericalLimits [{limitRadius=11.5, drivingBone {boneName=spine_01}, sphereLocalOffset=<0 0 0>}]
   }
   n25 = Anim Dynamics - Bone: dyn_tassel_front_02!(ComponentPose=@n24, Alpha=0.7, GravityOverride:Vector="0,0,0", ExternalForce="0,0,0") -> (Pose) {
    angularDampingOverride=1
    bGravityOverrideInSimSpace=true
    bOverrideAngularDamping=true
    bUseGravityOverride=true
    bUseSphericalLimits=true
    simulationSpace=RootRelative
    alphaBoolBlend {blendOption=Linear}
    boundBone {boneName=dyn_tassel_front_02}
    componentAppliedLinearAccClamp=<5000 5000 5000>
    componentLinearAccScale=<0.5 0.5 0.5>
    componentLinearVelScale=<0.5 0.5 0.5>
    physicsBodyDefinitions [{
      sphereCollisionRadius=0
      boundBone {boneName=dyn_tassel_front_02}
      boxExtents=<10 20 20>
      constraintSetup {angularLimitsMax=<25 -12 25>, angularLimitsMin=<-25 -25 -25>, angularTarget=<0 0 0>}
      localJointOffset=<5 0 0>
     }]
    sphericalLimits [{limitRadius=11.5, drivingBone {boneName=spine_01}, sphereLocalOffset=<0 0 0>}]
   }
   n26 = Anim Dynamics - Bone: dyn_tassel_front_03!(ComponentPose=@n25, Alpha=0.7, GravityOverride:Vector="0,0,0", ExternalForce="0,0,0") -> (Pose) {
    angularDampingOverride=1
    bGravityOverrideInSimSpace=true
    bOverrideAngularDamping=true
    bUseGravityOverride=true
    bUseSphericalLimits=true
    simulationSpace=RootRelative
    alphaBoolBlend {blendOption=Linear}
    boundBone {boneName=dyn_tassel_front_03}
    componentAppliedLinearAccClamp=<5000 5000 5000>
    componentLinearAccScale=<0.5 0.5 0.5>
    componentLinearVelScale=<0.5 0.5 0.5>
    physicsBodyDefinitions [{
      sphereCollisionRadius=0
      boundBone {boneName=dyn_tassel_front_03}
      boxExtents=<20 20 10>
      constraintSetup {angularLimitsMax=<25 -12 25>, angularLimitsMin=<-25 -25 -25>, angularTarget=<0 0 0>}
      localJointOffset=<5 0 0>
     }]
    sphericalLimits [{limitRadius=11.5, drivingBone {boneName=spine_01}, sphereLocalOffset=<0 0 0>}]
   }
   n27 = Anim Dynamics - Bone: dyn_tassel_front_04!(ComponentPose=@n26, Alpha=0.7, GravityOverride:Vector="0,0,0", ExternalForce="0,0,0") -> (Pose) {
    angularDampingOverride=1
    bGravityOverrideInSimSpace=true
    bOverrideAngularDamping=true
    bUseGravityOverride=true
    bUseSphericalLimits=true
    simulationSpace=RootRelative
    alphaBoolBlend {blendOption=Linear}
    boundBone {boneName=dyn_tassel_front_04}
    componentAppliedLinearAccClamp=<5000 5000 5000>
    componentLinearAccScale=<0.5 0.5 0.5>
    componentLinearVelScale=<0.5 0.5 0.5>
    physicsBodyDefinitions [{
      sphereCollisionRadius=0
      boundBone {boneName=dyn_tassel_front_04}
      boxExtents=<20 20 10>
      constraintSetup {angularLimitsMax=<25 -20 25>, angularLimitsMin=<-25 -25 -25>, angularTarget=<0 0 0>}
      localJointOffset=<5 0 0>
     }]
    sphericalLimits [{limitRadius=11.5, drivingBone {boneName=spine_01}, sphereLocalOffset=<0 0 0>}]
   }
   n28 = RigidBody(ComponentPose=@n27, OverrideWorldGravity=Get GravityOverride()) {
    bOverrideWorldGravity=true
    overridePhysicsAsset=&/F_MED_Raspberry_FE_Physics
    simulationSpace=BaseBoneSpace
    alphaBoolBlend {blendOption=Linear}
    baseBoneRef {boneName=spine_01}
    componentAppliedLinearAccClamp=<3000 3000 3000>
    componentLinearAccScale=<0.3 0.3 0.4>
    componentLinearVelScale=<0.3 0.3 0.4>
    overrideWorldGravity {z=-980}
    simSpaceSettings {
     maxAngularAcceleration=1000
     maxAngularVelocity=3
     maxLinearAcceleration=700
     maxLinearVelocity=500
     velocityScaleZ=0
     externalLinearDragV=<0.2 0.2>
    }
   }
   n29 = RigidBody!(ComponentPose=@n28, bAlphaBoolEnabled:bool, OverrideWorldGravity="0,0,-980", ExternalForce) -> (Pose) {
    alphaInputType=Bool
    bClampLinearTranslationLimitToRefPose=true
    bOverrideWorldGravity=true
    overridePhysicsAsset=&F_MED_Raspberry_Physics
    simulationSpace=BaseBoneSpace
    alphaBoolBlend {blendOption=Linear}
    baseBoneRef {boneName=pelvis}
    componentAppliedLinearAccClamp=<2500 2500 2500>
    componentLinearAccScale=<0.3 0.3 0.4>
    componentLinearVelScale=<0.3 0.3 0.4>
    overrideWorldGravity {z=-980}
    simSpaceSettings {
     maxAngularAcceleration=1000
     maxAngularVelocity=2
     maxLinearAcceleration=700
     maxLinearVelocity=500
     velocityScaleZ=0
     externalLinearDragV=<0.2 0.2 0.2>
    }
   }
   n30 = "Transform (Modify) Bone - Bone: deform_glute_l"(ComponentPose=@n29, Alpha=Get thigh_rot_otr_l() -> (thigh_rot_otr_l:double), Translation="-1,0,3", Rotation="0,0,0", Scale="1,1,1") {
    rotationMode=BMM_Additive
    rotationSpace=BCS_BoneSpace
    scaleMode=BMM_Additive
    scaleSpace=BCS_BoneSpace
    translationMode=BMM_Additive
    translationSpace=BCS_BoneSpace
    alphaBoolBlend {blendInTime=0.1, blendOption=Linear, blendOutTime=0.1}
    alphaScaleBiasClamp {bInterpResult=true}
    boneToModify {boneName=deform_glute_l}
    translation {x=-1, z=3}
   }
   n31 = "Transform (Modify) Bone - Bone: deform_glute_l"(ComponentPose=@n30, Alpha=Get thigh_rot_up_r() -> (thigh_rot_up_r:double), Translation="3,0,3", Rotation="0,0,0", Scale="1,1,1") {
    rotationMode=BMM_Additive
    rotationSpace=BCS_BoneSpace
    scaleMode=BMM_Additive
    scaleSpace=BCS_BoneSpace
    translationMode=BMM_Additive
    translationSpace=BCS_BoneSpace
    alphaBoolBlend {blendInTime=0.1, blendOption=Linear, blendOutTime=0.1}
    alphaScaleBiasClamp {bInterpResult=true}
    boneToModify {boneName=deform_glute_l}
    translation {x=3, z=3}
   }
   n32 = "Transform (Modify) Bone - Bone: deform_glute_l"!(
    ComponentPose=@n31
    bAlphaBoolEnabled:bool=Get CrouchingInPlace() -> (CrouchingInPlace:bool)
    Translation="-1,0,0"
    Rotation="0,0,0"
    Scale="1.1,1,1"
   ) -> (Pose) {
    alphaInputType=Bool
    rotationMode=BMM_Additive
    rotationSpace=BCS_BoneSpace
    scaleMode=BMM_Additive
    scaleSpace=BCS_BoneSpace
    translationMode=BMM_Additive
    translationSpace=BCS_BoneSpace
    alphaBoolBlend {blendInTime=0.1, blendOption=Linear, blendOutTime=0.1}
    alphaScaleBiasClamp {bInterpResult=true}
    boneToModify {boneName=deform_glute_l}
    scale {x=1.1}
    translation {x=-1}
   }
   n33 = Get FrontEndIdle() -> (FrontEndIdle:bool)
   n34 = "Transform (Modify) Bone - Bone: deform_glute_l"(ComponentPose=@n32, bAlphaBoolEnabled=@n33, Translation="-0.902,-1.720,-0.120", Rotation="0,0,0", Scale="1.400,1.409,1.000") {
    alphaInputType=Bool
    rotationMode=BMM_Additive
    rotationSpace=BCS_BoneSpace
    scaleMode=BMM_Additive
    scaleSpace=BCS_BoneSpace
    translationMode=BMM_Additive
    translationSpace=BCS_BoneSpace
    alphaBoolBlend {blendInTime=0.1, blendOption=Linear, blendOutTime=0.1}
    boneToModify {boneName=deform_glute_l}
    scale=<1.4 1.409>
    translation=<-0.902 -1.72 -0.12>
   }
   n35 = "Transform (Modify) Bone - Bone: deform_glute_r"!(ComponentPose=@n34, bAlphaBoolEnabled:bool=@n33, Translation="0.026,-1.627,-1.456", Rotation="0,0,0", Scale="2.020,1.441,1.000") -> (Pose) {
    alphaInputType=Bool
    rotationMode=BMM_Additive
    rotationSpace=BCS_BoneSpace
    scaleMode=BMM_Additive
    scaleSpace=BCS_BoneSpace
    translationMode=BMM_Additive
    translationSpace=BCS_BoneSpace
    alphaBoolBlend {blendInTime=0.1, blendOption=Linear, blendOutTime=0.1}
    boneToModify {boneName=deform_glute_r}
    scale=<2.0199 1.441>
    translation=<0.026 -1.627 -1.456>
   }
   Output Pose(Result=Component To Local(ComponentPose=@n35)) {layerGroup=None}
   n36 = Control Rig(Source=Copy Pose From Mesh() {bUseAttachedParent=true}) {
    alphaBoolBlend {blendOption=Linear}
    controlRigAssetReference {blueprintRigClass=&/F_Raspberry_DeformRig_Master_ControlRig.F_Raspberry_DeformRig_Master_ControlRig_C}
   }
   n37 = "Constraint - dyn_TopZipper_MSTR (PPPPPP)"!(
    ComponentPose=Local To Component(LocalPose=@n36)
    Alpha=1
    ConstraintWeights_0=0.0057
    ConstraintWeights_1:float=0.9128
    ConstraintWeights_2:float=0.0071
    ConstraintWeights_3:float=0.0034
    ConstraintWeights_4:float=0.0661
    ConstraintWeights_5:float=0.0049
   ) -> (Pose) {
    alphaBoolBlend {blendOption=Linear}
    boneToModify {boneName=dyn_TopZipper_MSTR}
    constraintSetup [
     {targetBone {boneName=neck_01}}
     {targetBone {boneName=spine_05}}
     {targetBone {boneName=clavicle_l}}
     {targetBone {boneName=clavicle_r}}
     {targetBone {boneName=spine_04}}
     {targetBone {boneName=spine_03}}
    ]
    constraintWeights [1,1,1,1,1,1]
   }
   n38 = "Constraint - dyn_l_armStrap_MSTR (PPPP)"!(
    ComponentPose=@n37
    Alpha=1
    ConstraintWeights_0=0.808
    ConstraintWeights_1:float=0.089
    ConstraintWeights_2:float=0.0933
    ConstraintWeights_3:float=0.0097
   ) -> (Pose) {
    alphaBoolBlend {blendOption=Linear}
    boneToModify {boneName=dyn_l_armStrap_MSTR}
    constraintSetup [
     {targetBone {boneName=upperarm_twist_01_l}}
     {targetBone {boneName=upperarm_l}}
     {targetBone {boneName=upperarm_twist_02_l}}
     {targetBone {boneName=deform_upperarm_bk_l}}
    ]
    constraintWeights [1,1,1,1]
   }
   n39 = "Constraint - dyn_tassel_front_01_MSTR (PPPPP)"!(
    ComponentPose=@n38
    Alpha=1
    ConstraintWeights_0=0.064
    ConstraintWeights_1:float=0.0322
    ConstraintWeights_2:float=0.3734
    ConstraintWeights_3:float=0.021
    ConstraintWeights_4:float=0.5094
   ) -> (Pose) {
    alphaBoolBlend {blendOption=Linear}
    boneToModify {boneName=dyn_tassel_front_01_MSTR}
    constraintSetup [
     {targetBone {boneName=spine_04}}
     {targetBone {boneName=spine_03}}
     {targetBone {boneName=spine_02}}
     {targetBone {boneName=spine_01}}
     {targetBone {boneName=pelvis}}
    ]
    constraintWeights [1,1,1,1,1]
   }
   n40 = "Constraint - dyn_tassel_front_03_MSTR (PPPPP)"!(
    ComponentPose=@n39
    Alpha=1
    ConstraintWeights_0=0.0667
    ConstraintWeights_1:float=0.0329
    ConstraintWeights_2:float=0.3258
    ConstraintWeights_3:float=0.0133
    ConstraintWeights_4:float=0.5612
   ) -> (Pose) {
    alphaBoolBlend {blendOption=Linear}
    boneToModify {boneName=dyn_tassel_front_03_MSTR}
    constraintSetup [
     {targetBone {boneName=spine_04}}
     {targetBone {boneName=spine_03}}
     {targetBone {boneName=spine_02}}
     {targetBone {boneName=spine_01}}
     {targetBone {boneName=pelvis}}
    ]
    constraintWeights [1,1,1,1,1]
   }
   n41 = "Constraint - dyn_BTMZipper_MSTR (PPPPP)"!(
    ComponentPose=@n40
    Alpha=1
    ConstraintWeights_0=0.102
    ConstraintWeights_1:float=0.0489
    ConstraintWeights_2:float=0.3835
    ConstraintWeights_3:float=0.0199
    ConstraintWeights_4:float=0.4458
   ) -> (Pose) {
    alphaBoolBlend {blendOption=Linear}
    boneToModify {boneName=dyn_BTMZipper_MSTR}
    constraintSetup [
     {targetBone {boneName=spine_04}}
     {targetBone {boneName=spine_03}}
     {targetBone {boneName=spine_02}}
     {targetBone {boneName=spine_01}}
     {targetBone {boneName=pelvis}}
    ]
    constraintWeights [1,1,1,1,1]
   }
   n42 = "Constraint - dyn_r_armStrap_MSTR (PPPP)"!(
    ComponentPose=@n41
    Alpha=1
    ConstraintWeights_0=0.781
    ConstraintWeights_1:float=0.1034
    ConstraintWeights_2:float=0.1073
    ConstraintWeights_3:float=0.0083
   ) -> (Pose) {
    alphaBoolBlend {blendOption=Linear}
    boneToModify {boneName=dyn_r_armStrap_MSTR}
    constraintSetup [
     {targetBone {boneName=upperarm_twist_01_r}}
     {targetBone {boneName=upperarm_r}}
     {targetBone {boneName=upperarm_twist_02_r}}
     {targetBone {boneName=deform_upperarm_bk_r}}
    ]
    constraintWeights [1,1,1,1]
   }
   n43 = "Constraint - dyn_tassel_front_02_MSTR (PPPPP)"!(
    ComponentPose=@n42
    Alpha=1
    ConstraintWeights_0=0.0708
    ConstraintWeights_1:float=0.0347
    ConstraintWeights_2:float=0.3404
    ConstraintWeights_3:float=0.0167
    ConstraintWeights_4:float=0.5374
   ) -> (Pose) {
    alphaBoolBlend {blendOption=Linear}
    boneToModify {boneName=dyn_tassel_front_02_MSTR}
    constraintSetup [
     {targetBone {boneName=spine_04}}
     {targetBone {boneName=spine_03}}
     {targetBone {boneName=spine_02}}
     {targetBone {boneName=spine_01}}
     {targetBone {boneName=pelvis}}
    ]
    constraintWeights [1,1,1,1,1]
   }
   n44 = "Constraint - dyn_tassel_front_04_MSTR (PPPPPP)"!(
    ComponentPose=@n43
    Alpha=1
    ConstraintWeights_0=0.0476
    ConstraintWeights_1:float=0.0241
    ConstraintWeights_2:float=0.318
    ConstraintWeights_3:float=0.0162
    ConstraintWeights_4:float=0.5941
    ConstraintWeights_5:float=0
   ) -> (Pose) {
    alphaBoolBlend {blendOption=Linear}
    boneToModify {boneName=dyn_tassel_front_04_MSTR}
    constraintSetup [
     {targetBone {boneName=spine_04}}
     {targetBone {boneName=spine_03}}
     {targetBone {boneName=spine_02}}
     {targetBone {boneName=spine_01}}
     {targetBone {boneName=pelvis}}
     {targetBone {boneName=deform_groin_l}}
    ]
    constraintWeights [1,1,1,1,1,1]
   }
   Save cached pose 'PreSim'(Pose=Component To Local(ComponentPose=@n44)) {cacheName=PreSim}
  ]}
 EventGraph {nodes [
   Event BlueprintUpdateAnimation()
   > n1 = Sequence!(execute) -> (then_0, then_1, then_2:Exec)
   n2 = Get bIsAccelerating2D()
   Set CrouchingInPlace(execute=@n1.then_1, CrouchingInPlace:bool=AND Boolean(A=Get bIsCrouching(), B=NOT Boolean(A=@n2))) -> (Output_Get:bool)
   n3 = Get bIsSkyDiving()
   > Set FrontEndIdle(FrontEndIdle:bool=AND Boolean!(A=Get bIsInFrontEnd(), B=false, C:bool=NOT Boolean(A=@n3)) -> (ReturnValue)) -> (Output_Get:bool)
   n4 = Gravity Override State Function DEPRECATED(
    execute=@n1.then_0
    CharacterPartAnimInstance=Self-Reference()
    GameTimeInSeconds=GetGameTimeInSeconds()
    GravityOverrideParams=Get Gravity Override Params()
   )
   > Set GravityOverride(GravityOverride=@n4)
   n5 = GetOwnerSkeletalMeshComponent()
   n6 = BreakRotator(InRot=BreakTransform(InTransform=GetDeltaTransformFromRefPose(self=@n5, BoneName=thigh_l, BaseName=pelvis)).Rotation)
   Set thigh_rot_up_l(execute=@n1.then_2, thigh_rot_up_l:double=MapRangeClamped(Value=@n6.Yaw, InRangeA=-10, InRangeB=-100)) -> (Output_Get:double)
   > Set thigh_rot_inr_l(thigh_rot_inr_l:double=MapRangeClamped(Value=@n6.Pitch, InRangeA=0, InRangeB=-10, OutRangeA=1, OutRangeB=0)) -> (Output_Get:double)
   > Set thigh_rot_otr_l(thigh_rot_otr_l:double=MapRangeClamped(Value=@n6.Pitch, InRangeA=20, InRangeB=50)) -> (Output_Get:double)
   n7 = BreakRotator(InRot=BreakTransform(InTransform=GetDeltaTransformFromRefPose(self=@n5, BoneName=thigh_r, BaseName=pelvis)).Rotation)
   > Set thigh_rot_up_r(thigh_rot_up_r:double=MapRangeClamped(Value=@n7.Yaw, InRangeA=-15, InRangeB=-100)) -> (Output_Get:double)
   > Set thigh_rot_inr_r(thigh_rot_inr_r:double=MapRangeClamped(Value=@n7.Pitch, InRangeA=0, InRangeB=-10, OutRangeA=1, OutRangeB=0)) -> (Output_Get:double)
   > Set thigh_rot_otr_r(thigh_rot_otr_r:double=MapRangeClamped(Value=@n7.Pitch, InRangeA=20, InRangeB=50)) -> (Output_Get:double)
   NOT Boolean(A=Get bIsPlayingEmote())
  ]}
]
variables [
 CrouchingInPlace {type=boolean}
 FrontEndIdle {type=boolean}
 Gravity Override Params {type=GravityOverrideParamsStruct}
 GravityOverride {type=Vector}
 thigh_rot_inr_l {type=number}
 thigh_rot_inr_r {type=number}
 thigh_rot_otr_l {type=number}
 thigh_rot_otr_r {type=number}
 thigh_rot_up_l {type=number}
 thigh_rot_up_r {type=number}
]
