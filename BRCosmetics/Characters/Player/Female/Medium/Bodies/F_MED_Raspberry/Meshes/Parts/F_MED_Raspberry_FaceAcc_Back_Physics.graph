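# Physics asset for the back face-accessory mesh of the F_MED_Raspberry female player body, set up for hanging braid motion.
# Five capsule chains (center back, left and right back, left and right side) each start at a tiny MSTR body jointed to the head.
# Braid segments have tight swing and twist limits, looser toward the side-braid tips.
# Capsules on head, neck, spine, clavicles and arms form a body proxy with free angular joints.
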PhysicsAsset /BRCosmetics/Characters/Player/Female/Medium/Bodies/F_MED_Raspberry/Meshes/Parts/F_MED_Raspberry_FaceAcc_Back_Physics
bodies {
 clavicle_l [{
   shapeType=Capsule
   length=5.1994
   radius=9.3006
   shapeName=clavicle_r_capsule
   center=<5.6741 -0.8778 -3.1114>
   rotation=<rot 75.1565 -0.0002 -0.0002>
  }]
 clavicle_r [{
   shapeType=Capsule
   length=4.7655
   radius=9.7345
   shapeName=clavicle_r_capsule
   center=<-4.6188 0 1.5478>
   rotation=<rot 75.1565 -0.0002 -0.0002>
  }]
 dyn_back_braid_01 [{shapeType=Capsule, length=5.583, radius=2.2864, shapeName=dyn_back_braid_01_capsule, center=<4.1563 0 2>, rotation=<rot 90 0 0>}]
 dyn_back_braid_02 [{shapeType=Capsule, length=5.7354, radius=2.134, shapeName=dyn_back_braid_02_capsule, center=<3.8259 0 2>, rotation=<rot 90 0 0>}]
 dyn_back_braid_03 [{shapeType=Capsule, length=7.5313, radius=2.6634, shapeName=dyn_back_braid_03_capsule, center=<5.3054 0 2>, rotation=<rot 90 0 0>}]
 dyn_back_braid_04 [{shapeType=Capsule, length=6.1546, radius=3.0654, shapeName=dyn_back_braid_04_capsule, center=<4.7221 0 2>, rotation=<rot 90 0 0>}]
 dyn_back_braid_MSTR [{shapeType=Capsule, length=0.505, radius=0.505, shapeName=dyn_back_braid_MSTR_capsule}]
 dyn_l_back_braid_01 [{shapeType=Capsule, length=5.583, radius=2.2864, shapeName=dyn_l_back_braid_01_capsule, center=<3.8259 0 2>, rotation=<rot 90 0 0>}]
 dyn_l_back_braid_02 [{shapeType=Capsule, length=5.583, radius=2.2864, shapeName=dyn_l_back_braid_02_capsule, center=<4.53 0 2>, rotation=<rot 90 0 0>}]
 dyn_l_back_braid_03 [{shapeType=Capsule, length=5.7354, radius=2.134, shapeName=dyn_l_back_braid_03_capsule, center=<3.8259 0 2>, rotation=<rot 90 0 0>}]
 dyn_l_back_braid_04 [{shapeType=Capsule, length=6.7178, radius=2.5259, shapeName=dyn_l_back_braid_04_capsule, center=<4.957 0 2>, rotation=<rot 90 0 0>}]
 dyn_l_back_braid_05 [{shapeType=Capsule, length=6.1546, radius=3.0654, shapeName=dyn_l_back_braid_05_capsule, center=<4.7221 0 2>, rotation=<rot 90 0 0>}]
 dyn_l_back_braid_MSTR [{shapeType=Capsule, length=0.505, radius=0.505, shapeName=dyn_l_back_braid_MSTR_capsule}]
 dyn_l_side_b_braid_01 [{shapeType=Capsule, length=6.4445, radius=1.4249, shapeName=dyn_l_side_b_braid_01_capsule, center=<3.8259 0 2>, rotation=<rot 90 0 0>}]
 dyn_l_side_b_braid_02 [{shapeType=Capsule, length=6.4445, radius=1.4249, shapeName=dyn_l_side_b_braid_02_capsule, center=<3.8259 0 2>, rotation=<rot 90 0 0>}]
 dyn_l_side_b_braid_03 [{shapeType=Capsule, length=5.9005, radius=1.9689, shapeName=dyn_l_side_b_braid_03_capsule, center=<3.8259 0 2>, rotation=<rot 90 0 0>}]
 dyn_l_side_b_braid_04 [{shapeType=Capsule, length=6.103, radius=1.7664, shapeName=dyn_l_side_b_braid_04_capsule, center=<3.8259 0 1.5136>, rotation=<rot 90 0 0>}]
 dyn_l_side_b_braid_05 [{shapeType=Capsule, length=6.4564, radius=2.7636, shapeName=dyn_l_side_b_braid_05_capsule, center=<4.7221 0 2>, rotation=<rot 90 0 0>}]
 dyn_l_side_b_braid_MSTR [{shapeType=Capsule, length=0.505, radius=0.505, shapeName=dyn_l_side_b_braid_MSTR_capsule}]
 dyn_r_back_braid_01 [{shapeType=Capsule, length=5.583, radius=2.2864, shapeName=dyn_r_back_braid_01_capsule, center=<3.8259 0 2>, rotation=<rot 90 0 0>}]
 dyn_r_back_braid_02 [{shapeType=Capsule, length=5.583, radius=2.2864, shapeName=dyn_r_back_braid_02_capsule, center=<4.2431 0 2>, rotation=<rot 90 0 0>}]
 dyn_r_back_braid_03 [{shapeType=Capsule, length=5.7354, radius=2.134, shapeName=dyn_r_back_braid_03_capsule, center=<3.8259 0 2>, rotation=<rot 90 0 0>}]
 dyn_r_back_braid_04 [{shapeType=Capsule, length=7.5023, radius=1.854, shapeName=dyn_r_back_braid_04_capsule, center=<4.8581 0 1.4418>, rotation=<rot 90 0 0>}]
 dyn_r_back_braid_05 [{shapeType=Capsule, length=6.1546, radius=3.0654, shapeName=dyn_r_back_braid_05_capsule, center=<4.7221 0 2>, rotation=<rot 90 0 0>}]
 dyn_r_back_braid_MSTR [{shapeType=Capsule, length=0.505, radius=0.505, shapeName=dyn_r_back_braid_MSTR_capsule}]
 dyn_r_side_b_braid_01 [{shapeType=Capsule, length=6.7116, radius=1.1578, shapeName=dyn_r_side_b_braid_01_capsule, center=<3.8259 0 2>, rotation=<rot 90 0 0>}]
 dyn_r_side_b_braid_02 [{shapeType=Capsule, length=6.7116, radius=1.1578, shapeName=dyn_r_side_b_braid_02_capsule, center=<3.8259 0 2>, rotation=<rot 90 0 0>}]
 dyn_r_side_b_braid_03 [{shapeType=Capsule, length=6.5067, radius=1.3627, shapeName=dyn_r_side_b_braid_03_capsule, center=<3.8259 0 2>, rotation=<rot 90 0 0>}]
 dyn_r_side_b_braid_04 [{shapeType=Capsule, length=7.8585, radius=1.5208, shapeName=dyn_r_side_b_braid_04_capsule, center=<4.6094 0 2>, rotation=<rot 90 0 0>}]
 dyn_r_side_b_braid_05 [{shapeType=Capsule, length=7.3781, radius=1.842, shapeName=dyn_r_side_b_braid_05_capsule, center=<4.7221 0 2>, rotation=<rot 90 0 0>}]
 dyn_r_side_b_braid_MSTR [{shapeType=Capsule, length=0.505, radius=0.505, shapeName=dyn_r_side_b_braid_MSTR_capsule}]
 head [{shapeType=Capsule, length=7.2769, radius=8.2628, rotation=<rot 90 0 0>}]
 lowerarm_l [{shapeType=Capsule, length=21.5747, radius=4.3914, shapeName=lowerarm_l_capsule, center=<9.4537 0 0>, rotation=<rot 90 0 0>}]
 lowerarm_r [{shapeType=Capsule, length=20.9949, radius=4.9712, shapeName=lowerarm_l_capsule, center=<-11.7853 0 0>, rotation=<rot 90 0 0>}]
 neck_01 [
  {shapeType=Capsule, length=58.6982, radius=4.3018, shapeName=neck_01_capsule}
  {
   shapeType=Capsule
   length=17.5965
   radius=6.4771
   shapeName=neck_01_capsule
   center=<0.9856 -7.6367 18.3357>
   rotation=<rot 51.4133 -47.9896 -60.7786>
  }
 ]
 neck_02 [{shapeType=Capsule, length=57.7917, radius=4.2083, shapeName=neck_01_capsule, center=<2.8325 -0.5621 0>}]
 spine_02 [{shapeType=Capsule, length=7.4682, radius=11.582, shapeName=spine_02_capsule}]
 spine_03 [{shapeType=Capsule, length=12.75, radius=7, shapeName=spine_04_capsule, center=<-0.8829 1.2598 0>}]
 spine_04 [{shapeType=Capsule, length=5.5833, radius=13.096, shapeName=spine_04_capsule, center=<0 1.6532 0>}]
 spine_05 [{shapeType=Capsule, length=12.7251, radius=10.4303, shapeName=spine_05_capsule, center=<0 -1.3547 0>, rotation=<rot 0 0 -5.1877>}]
 upperarm_l [{shapeType=Capsule, length=20.3616, radius=4.6384, shapeName=upperarm_l_capsule, center=<12 -0.8 0>, rotation=<rot 90 0 0>}]
 upperarm_r [{shapeType=Capsule, length=20.3616, radius=5.1384, shapeName=upperarm_r_capsule, center=<-11.9128 0.8 0>, rotation=<rot 90 0 0>}]
}
constraints [
 {
  bone1Name=dyn_r_side_b_braid_MSTR
  bone2Name=head
  swing1LimitDegrees=45
  swing1Motion=Free
  swing2LimitDegrees=45
  swing2Motion=Free
  twistLimitDegrees=45
  twistMotion=Free
 }
 {
  bone1Name=dyn_l_side_b_braid_MSTR
  bone2Name=head
  swing1LimitDegrees=45
  swing1Motion=Free
  swing2LimitDegrees=45
  swing2Motion=Free
  twistLimitDegrees=45
  twistMotion=Free
 }
 {
  bone1Name=dyn_l_back_braid_MSTR
  bone2Name=head
  swing1LimitDegrees=45
  swing1Motion=Free
  swing2LimitDegrees=45
  swing2Motion=Free
  twistLimitDegrees=45
  twistMotion=Free
 }
 {
  bone1Name=dyn_r_back_braid_MSTR
  bone2Name=head
  swing1LimitDegrees=45
  swing1Motion=Free
  swing2LimitDegrees=45
  swing2Motion=Free
  twistLimitDegrees=45
  twistMotion=Free
 }
 {
  bone1Name=dyn_back_braid_MSTR
  bone2Name=head
  swing1LimitDegrees=45
  swing1Motion=Free
  swing2LimitDegrees=45
  swing2Motion=Free
  twistLimitDegrees=45
  twistMotion=Free
 }
 {bone1Name=dyn_r_side_b_braid_01, bone2Name=dyn_r_side_b_braid_MSTR, swing1LimitDegrees=10, swing2LimitDegrees=10, twistLimitDegrees=5}
 {bone1Name=dyn_r_side_b_braid_02, bone2Name=dyn_r_side_b_braid_01, swing1LimitDegrees=10, swing2LimitDegrees=10, twistLimitDegrees=5}
 {bone1Name=dyn_r_side_b_braid_03, bone2Name=dyn_r_side_b_braid_02, swing1LimitDegrees=20, swing2LimitDegrees=20, twistLimitDegrees=20}
 {bone1Name=dyn_r_side_b_braid_04, bone2Name=dyn_r_side_b_braid_03, swing1LimitDegrees=20, swing2LimitDegrees=20, twistLimitDegrees=20}
 {bone1Name=dyn_r_side_b_braid_05, bone2Name=dyn_r_side_b_braid_04, swing1LimitDegrees=20, swing2LimitDegrees=20, twistLimitDegrees=20}
 {bone1Name=dyn_l_side_b_braid_01, bone2Name=dyn_l_side_b_braid_MSTR, swing1LimitDegrees=10, swing2LimitDegrees=10, twistLimitDegrees=5}
 {bone1Name=dyn_l_side_b_braid_02, bone2Name=dyn_l_side_b_braid_01, swing1LimitDegrees=10, swing2LimitDegrees=10, twistLimitDegrees=5}
 {bone1Name=dyn_l_side_b_braid_03, bone2Name=dyn_l_side_b_braid_02, swing1LimitDegrees=10, swing2LimitDegrees=10, twistLimitDegrees=5}
 {bone1Name=dyn_l_side_b_braid_04, bone2Name=dyn_l_side_b_braid_03, swing1LimitDegrees=15, swing2LimitDegrees=15, twistLimitDegrees=10}
 {bone1Name=dyn_l_side_b_braid_05, bone2Name=dyn_l_side_b_braid_04, swing1LimitDegrees=15, swing2LimitDegrees=15, twistLimitDegrees=10}
 {bone1Name=dyn_l_back_braid_01, bone2Name=dyn_l_back_braid_MSTR, swing1LimitDegrees=10, swing2LimitDegrees=10, twistLimitDegrees=5}
 {bone1Name=dyn_l_back_braid_02, bone2Name=dyn_l_back_braid_01, swing1LimitDegrees=10, swing2LimitDegrees=10, twistLimitDegrees=5}
 {bone1Name=dyn_l_back_braid_03, bone2Name=dyn_l_back_braid_02, swing1LimitDegrees=10, swing2LimitDegrees=10, twistLimitDegrees=5}
 {bone1Name=dyn_l_back_braid_04, bone2Name=dyn_l_back_braid_03, swing1LimitDegrees=10, swing2LimitDegrees=10, twistLimitDegrees=5}
 {bone1Name=dyn_l_back_braid_05, bone2Name=dyn_l_back_braid_04, swing1LimitDegrees=10, swing2LimitDegrees=10, twistLimitDegrees=5}
 {bone1Name=dyn_r_back_braid_01, bone2Name=dyn_r_back_braid_MSTR, swing1LimitDegrees=10, swing2LimitDegrees=10, twistLimitDegrees=5}
 {bone1Name=dyn_r_back_braid_02, bone2Name=dyn_r_back_braid_01, swing1LimitDegrees=10, swing2LimitDegrees=10, twistLimitDegrees=5}
 {bone1Name=dyn_r_back_braid_03, bone2Name=dyn_r_back_braid_02, swing1LimitDegrees=10, swing2LimitDegrees=10, twistLimitDegrees=5}
 {bone1Name=dyn_r_back_braid_04, bone2Name=dyn_r_back_braid_03, swing1LimitDegrees=10, swing2LimitDegrees=10, twistLimitDegrees=5}
 {bone1Name=dyn_r_back_braid_05, bone2Name=dyn_r_back_braid_04, swing1LimitDegrees=10, swing2LimitDegrees=10, twistLimitDegrees=5}
 {bone1Name=dyn_back_braid_01, bone2Name=dyn_back_braid_MSTR, swing1LimitDegrees=10, swing2LimitDegrees=10, twistLimitDegrees=5}
 {bone1Name=dyn_back_braid_02, bone2Name=dyn_back_braid_01, swing1LimitDegrees=10, swing2LimitDegrees=10, twistLimitDegrees=5}
 {bone1Name=dyn_back_braid_03, bone2Name=dyn_back_braid_02, swing1LimitDegrees=10, swing2LimitDegrees=10, twistLimitDegrees=5}
 {bone1Name=dyn_back_braid_04, bone2Name=dyn_back_braid_03, swing1LimitDegrees=10, swing2LimitDegrees=10, twistLimitDegrees=5}
 {
  bone1Name=head
  bone2Name=neck_02
  swing1LimitDegrees=45
  swing1Motion=Free
  swing2LimitDegrees=45
  swing2Motion=Free
  twistLimitDegrees=45
  twistMotion=Free
 }
 {
  bone1Name=spine_03
  bone2Name=spine_02
  swing1LimitDegrees=45
  swing1Motion=Free
  swing2LimitDegrees=45
  swing2Motion=Free
  twistLimitDegrees=45
  twistMotion=Free
 }
 {
  bone1Name=spine_04
  bone2Name=spine_03
  swing1LimitDegrees=45
  swing1Motion=Free
  swing2LimitDegrees=45
  swing2Motion=Free
  twistLimitDegrees=45
  twistMotion=Free
 }
 {
  bone1Name=spine_05
  bone2Name=spine_04
  swing1LimitDegrees=45
  swing1Motion=Free
  swing2LimitDegrees=45
  swing2Motion=Free
  twistLimitDegrees=45
  twistMotion=Free
 }
 {
  bone1Name=neck_01
  bone2Name=spine_05
  swing1LimitDegrees=45
  swing1Motion=Free
  swing2LimitDegrees=45
  swing2Motion=Free
  twistLimitDegrees=45
  twistMotion=Free
 }
 {
  bone1Name=neck_02
  bone2Name=neck_01
  swing1LimitDegrees=45
  swing1Motion=Free
  swing2LimitDegrees=45
  swing2Motion=Free
  twistLimitDegrees=45
  twistMotion=Free
 }
 {
  bone1Name=clavicle_l
  bone2Name=spine_05
  swing1LimitDegrees=45
  swing1Motion=Free
  swing2LimitDegrees=45
  swing2Motion=Free
  twistLimitDegrees=45
  twistMotion=Free
 }
 {
  bone1Name=upperarm_l
  bone2Name=clavicle_l
  swing1LimitDegrees=45
  swing1Motion=Free
  swing2LimitDegrees=45
  swing2Motion=Free
  twistLimitDegrees=45
  twistMotion=Free
 }
 {
  bone1Name=lowerarm_l
  bone2Name=upperarm_l
  swing1LimitDegrees=45
  swing1Motion=Free
  swing2LimitDegrees=45
  swing2Motion=Free
  twistLimitDegrees=45
  twistMotion=Free
 }
 {
  bone1Name=clavicle_r
  bone2Name=spine_05
  swing1LimitDegrees=45
  swing1Motion=Free
  swing2LimitDegrees=45
  swing2Motion=Free
  twistLimitDegrees=45
  twistMotion=Free
 }
 {
  bone1Name=upperarm_r
  bone2Name=clavicle_r
  swing1LimitDegrees=45
  swing1Motion=Free
  swing2LimitDegrees=45
  swing2Motion=Free
  twistLimitDegrees=45
  twistMotion=Free
 }
 {
  bone1Name=lowerarm_r
  bone2Name=upperarm_r
  swing1LimitDegrees=45
  swing1Motion=Free
  swing2LimitDegrees=45
  swing2Motion=Free
  twistLimitDegrees=45
  twistMotion=Free
 }
]
