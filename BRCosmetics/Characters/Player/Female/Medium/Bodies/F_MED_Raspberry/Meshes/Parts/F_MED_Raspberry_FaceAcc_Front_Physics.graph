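# Physics asset for the front face-accessory part of the female medium Raspberry character, covering four dangling braids: front and side, left and right.
# Each braid is a five-capsule chain on a small master body attached to the head, with limited angular joints whose limits mostly widen toward the tip.
# Head, neck, spine, clavicle and arm capsules provide the body shapes, linked by angularly free constraints.
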PhysicsAsset /BRCosmetics/Characters/Player/Female/Medium/Bodies/F_MED_Raspberry/Meshes/Parts/F_MED_Raspberry_FaceAcc_Front_Physics
bodies {
 clavicle_l [{
   shapeType=Capsule
   length=8.3407
   radius=6.9593
   shapeName=clavicle_r_capsule
   center=<4.1126 -1.4842 -1.8443>
   rotation=<rot 75.1565 -0.0002 -0.0002>
  }]
 clavicle_r [{shapeType=Capsule, length=7.5662, radius=7.8088, shapeName=clavicle_r_capsule, center=<-4.7535 0 2>, rotation=<rot 75.1565 -0.0002 0>}]
 dyn_l_fr_braid_01 [{shapeType=Capsule, length=3.4055, radius=1.1945, shapeName=dyn_l_fr_braid_01_capsule, center=<3 0 0.5>, rotation=<rot 90 0 0>}]
 dyn_l_fr_braid_02 [{shapeType=Capsule, length=3.5305, radius=1.3195, shapeName=dyn_l_fr_braid_02_capsule, center=<3 0 0.5>, rotation=<rot 90 0 0>}]
 dyn_l_fr_braid_03 [{shapeType=Capsule, length=3.2449, radius=1.3551, shapeName=dyn_l_fr_braid_03_capsule, center=<3.2 0 0>, rotation=<rot 90 0 0>}]
 dyn_l_fr_braid_04 [{shapeType=Capsule, length=3.1955, radius=1.6051, shapeName=dyn_l_fr_braid_04_capsule, center=<3 0 0>, rotation=<rot 90 0 0>}]
 dyn_l_fr_braid_05 [{shapeType=Capsule, length=2.6623, radius=1.7127, shapeName=dyn_l_fr_braid_05_capsule, center=<3.5 0 0>, rotation=<rot 90 0 0>}]
 dyn_l_fr_braid_MSTR [{shapeType=Capsule, length=0.505, radius=0.505, shapeName=dyn_l_fr_braid_MSTR_capsule}]
 dyn_l_side_a_braid_01 [{shapeType=Capsule, length=1.2141, radius=2.5109, shapeName=dyn_l_side_a_braid_01_capsule, center=<2.9715 0 0.56>, rotation=<rot 90 0 0>}]
 dyn_l_side_a_braid_02 [{shapeType=Capsule, length=1.9607, radius=2.3893, shapeName=dyn_l_side_a_braid_02_capsule, center=<3.0428 0 0>, rotation=<rot 90 0 0>}]
 dyn_l_side_a_braid_03 [{shapeType=Capsule, length=2.462, radius=2.138, shapeName=dyn_l_side_a_braid_03_capsule, center=<3.047 0 0>, rotation=<rot 90 0 0>}]
 dyn_l_side_a_braid_04 [{shapeType=Capsule, length=4.9473, radius=1.4027, shapeName=dyn_l_side_a_braid_04_capsule, center=<3.9425 0 0>, rotation=<rot 90 0 0>}]
 dyn_l_side_a_braid_05 [{shapeType=Capsule, length=4.1016, radius=1.8867, shapeName=dyn_l_side_a_braid_05_capsule, center=<4.1395 0 0>, rotation=<rot 90 0 0>}]
 dyn_l_side_a_braid_MSTR [{shapeType=Capsule, length=0.505, radius=0.505, shapeName=dyn_l_side_a_braid_MSTR_capsule}]
 dyn_r_fr_braid_01 [{shapeType=Capsule, length=3.5537, radius=1.2963, shapeName=dyn_r_fr_braid_01_capsule, center=<3 0 0.5>, rotation=<rot 90 0 0>}]
 dyn_r_fr_braid_02 [{shapeType=Capsule, length=3.5537, radius=1.1713, shapeName=dyn_r_fr_braid_02_capsule, center=<3.1243 0 0.5>, rotation=<rot 90 0 0>}]
 dyn_r_fr_braid_03 [{shapeType=Capsule, length=3.5537, radius=1.1713, shapeName=dyn_r_fr_braid_03_capsule, center=<3.252 0 0>, rotation=<rot 90 0 0>}]
 dyn_r_fr_braid_04 [{shapeType=Capsule, length=2.7543, radius=1.4213, shapeName=dyn_r_fr_braid_04_capsule, center=<3 0 0>, rotation=<rot 90 0 0>}]
 dyn_r_fr_braid_05 [{shapeType=Capsule, length=0.5, radius=2.0663, shapeName=dyn_r_fr_braid_05_capsule, center=<2 0 0>, rotation=<rot 90 0 0>}]
 dyn_r_fr_braid_MSTR [{shapeType=Capsule, length=0.505, radius=0.505, shapeName=dyn_r_fr_braid_MSTR_capsule}]
 dyn_r_side_a_braid_01 [{shapeType=Capsule, length=0.7398, radius=2.8852, shapeName=dyn_r_side_a_braid_01_capsule, center=<3 0 1>, rotation=<rot 90 0 0>}]
 dyn_r_side_a_braid_02 [{shapeType=Capsule, length=1.6547, radius=2.4703, shapeName=dyn_r_side_a_braid_02_capsule, center=<3 0 1>, rotation=<rot 90 0 0>}]
 dyn_r_side_a_braid_03 [{shapeType=Capsule, length=2.5674, radius=2.3076, shapeName=dyn_r_side_a_braid_03_capsule, center=<3 0 1>, rotation=<rot 90 0 0>}]
 dyn_r_side_a_braid_04 [{shapeType=Capsule, length=4.75, radius=1.375, shapeName=dyn_r_side_a_braid_04_capsule, center=<4 0 1>, rotation=<rot 90 0 0>}]
 dyn_r_side_a_braid_05 [{shapeType=Capsule, length=4.25, radius=1.75, shapeName=dyn_r_side_a_braid_05_capsule, center=<4 0 1>, rotation=<rot 90 0 0>}]
 dyn_r_side_a_braid_MSTR [{shapeType=Capsule, length=0.505, radius=0.505, shapeName=dyn_r_side_a_braid_MSTR_capsule}]
 head [
  {shapeType=Capsule, length=3.4019, radius=6.2628, center=<3.1939 1.6506 0>, rotation=<rot 90 0 0>}
  {shapeType=Capsule, length=1, radius=7.5, shapeName=head_capsule, center=<3.9645 -0.6188 0>, rotation=<rot 90 0 0>}
 ]
 lowerarm_l [{shapeType=Capsule, length=21.5747, radius=4.3914, shapeName=lowerarm_l_capsule, center=<9.4537 0 0>, rotation=<rot 90 0 0>}]
 lowerarm_r [{shapeType=Capsule, length=17.8699, radius=4.2212, shapeName=lowerarm_l_capsule, center=<-11.1763 0 0>, rotation=<rot 90 0 0>}]
 neck_01 [{shapeType=Capsule, length=5.3232, radius=4.3018, shapeName=neck_01_capsule}]
 neck_02 [{shapeType=Capsule, length=5.3232, radius=3.3018, shapeName=neck_01_capsule, center=<0 -0.5621 0>}]
 spine_02 [{shapeType=Capsule, length=2.9682, radius=11.582, shapeName=spine_02_capsule}]
 spine_03 [{shapeType=Capsule, length=12.75, radius=7, shapeName=spine_04_capsule, center=<-0.8829 1.2598 0>}]
 spine_04 [{shapeType=Capsule, length=5.7595, radius=11.4709, shapeName=spine_04_capsule, center=<-1.0334 1.5028 0>}]
 spine_05 [{shapeType=Capsule, length=12.6201, radius=8.8364, shapeName=spine_05_capsule, center=<0 -1.3547 0>}]
 upperarm_l [{shapeType=Capsule, length=20.3616, radius=4.6384, shapeName=upperarm_l_capsule, center=<12 -0.8 0>, rotation=<rot 90 0 0>}]
 upperarm_r [{shapeType=Capsule, length=19.8616, radius=5.3884, shapeName=upperarm_r_capsule, center=<-11.9128 0.8 0>, rotation=<rot 90 0 0>}]
 upperarm_twist_01_l [{shapeType=Capsule, length=6.38, radius=4.38, shapeName=upperarm_twist_01_l_capsule, rotation=<rot 90 0 0>}]
 upperarm_twist_01_r [{
   shapeType=Capsule
   length=6.005
   radius=5.88
   shapeName=upperarm_twist_01_r_capsule
   center=<4.1693 1.4888 -0.82>
   rotation=<rot 85 -10.0002 -10.0004>
  }]
}
constraints [
 {
  bone1Name=dyn_l_side_a_braid_MSTR
  bone2Name=head
  swing1LimitDegrees=45
  swing1Motion=Free
  swing2LimitDegrees=45
  swing2Motion=Free
  twistLimitDegrees=45
  twistMotion=Free
 }
 {
  bone1Name=dyn_r_fr_braid_MSTR
  bone2Name=head
  swing1LimitDegrees=45
  swing1Motion=Free
  swing2LimitDegrees=45
  swing2Motion=Free
  twistLimitDegrees=45
  twistMotion=Free
 }
 {
  bone1Name=dyn_l_fr_braid_MSTR
  bone2Name=head
  swing1LimitDegrees=45
  swing1Motion=Free
  swing2LimitDegrees=45
  swing2Motion=Free
  twistLimitDegrees=45
  twistMotion=Free
 }
 {
  bone1Name=dyn_r_side_a_braid_MSTR
  bone2Name=head
  swing1LimitDegrees=45
  swing1Motion=Free
  swing2LimitDegrees=45
  swing2Motion=Free
  twistLimitDegrees=45
  twistMotion=Free
 }
 {bone1Name=dyn_l_side_a_braid_01, bone2Name=dyn_l_side_a_braid_MSTR, swing1LimitDegrees=15, swing2LimitDegrees=35, twistLimitDegrees=10}
 {bone1Name=dyn_l_side_a_braid_02, bone2Name=dyn_l_side_a_braid_01, swing1LimitDegrees=15, swing2LimitDegrees=35, twistLimitDegrees=10}
 {bone1Name=dyn_l_side_a_braid_03, bone2Name=dyn_l_side_a_braid_02, swing1LimitDegrees=15, swing2LimitDegrees=45, twistLimitDegrees=15}
 {bone1Name=dyn_l_side_a_braid_04, bone2Name=dyn_l_side_a_braid_03, swing1LimitDegrees=15, swing2LimitDegrees=45, twistLimitDegrees=15}
 {bone1Name=dyn_l_side_a_braid_05, bone2Name=dyn_l_side_a_braid_04, swing1LimitDegrees=15, swing2LimitDegrees=45, twistLimitDegrees=15}
 {bone1Name=dyn_r_fr_braid_01, bone2Name=dyn_r_fr_braid_MSTR, swing1LimitDegrees=5, swing2LimitDegrees=5, twistLimitDegrees=5}
 {bone1Name=dyn_r_fr_braid_02, bone2Name=dyn_r_fr_braid_01, swing1LimitDegrees=5, swing2LimitDegrees=10, twistLimitDegrees=5}
 {bone1Name=dyn_r_fr_braid_03, bone2Name=dyn_r_fr_braid_02, swing1LimitDegrees=10, swing2LimitDegrees=20, twistLimitDegrees=20}
 {bone1Name=dyn_r_fr_braid_04, bone2Name=dyn_r_fr_braid_03, swing1LimitDegrees=20, swing2LimitDegrees=35, twistLimitDegrees=15}
 {bone1Name=dyn_r_fr_braid_05, bone2Name=dyn_r_fr_braid_04, swing1LimitDegrees=20, swing2LimitDegrees=35, twistLimitDegrees=15}
 {bone1Name=dyn_l_fr_braid_01, bone2Name=dyn_l_fr_braid_MSTR, swing1LimitDegrees=5, swing2LimitDegrees=5, twistLimitDegrees=5}
 {bone1Name=dyn_l_fr_braid_02, bone2Name=dyn_l_fr_braid_01, swing1LimitDegrees=5, swing2LimitDegrees=10, twistLimitDegrees=5}
 {bone1Name=dyn_l_fr_braid_03, bone2Name=dyn_l_fr_braid_02, swing1LimitDegrees=10, swing2LimitDegrees=15, twistLimitDegrees=10}
 {bone1Name=dyn_l_fr_braid_04, bone2Name=dyn_l_fr_braid_03, swing1LimitDegrees=15, swing2LimitDegrees=35, twistLimitDegrees=15}
 {bone1Name=dyn_l_fr_braid_05, bone2Name=dyn_l_fr_braid_04, swing1LimitDegrees=20, swing2LimitDegrees=45, twistLimitDegrees=15}
 {bone1Name=dyn_r_side_a_braid_01, bone2Name=dyn_r_side_a_braid_MSTR, swing1LimitDegrees=15, swing2LimitDegrees=35, twistLimitDegrees=10}
 {bone1Name=dyn_r_side_a_braid_02, bone2Name=dyn_r_side_a_braid_01, swing1LimitDegrees=15, swing2LimitDegrees=35, twistLimitDegrees=10}
 {bone1Name=dyn_r_side_a_braid_03, bone2Name=dyn_r_side_a_braid_02, swing1LimitDegrees=15, swing2LimitDegrees=45, twistLimitDegrees=5}
 {bone1Name=dyn_r_side_a_braid_04, bone2Name=dyn_r_side_a_braid_03, swing1LimitDegrees=15, swing2LimitDegrees=45, twistLimitDegrees=5}
 {bone1Name=dyn_r_side_a_braid_05, bone2Name=dyn_r_side_a_braid_04, swing1LimitDegrees=15, swing2LimitDegrees=45, twistLimitDegrees=5}
 {
  bone1Name=head
  bone2Name=neck_02
  swing1LimitDegrees=45
  swing1Motion=Free
  swing2LimitDegrees=45
  swing2Motion=Free
  twistLimitDegrees=45
  twistMotion=Free
 }
 {
  bone1Name=spine_03
  bone2Name=spine_02
  swing1LimitDegrees=45
  swing1Motion=Free
  swing2LimitDegrees=45
  swing2Motion=Free
  twistLimitDegrees=45
  twistMotion=Free
 }
 {
  bone1Name=spine_04
  bone2Name=spine_03
  swing1LimitDegrees=45
  swing1Motion=Free
  swing2LimitDegrees=45
  swing2Motion=Free
  twistLimitDegrees=45
  twistMotion=Free
 }
 {
  bone1Name=spine_05
  bone2Name=spine_04
  swing1LimitDegrees=45
  swing1Motion=Free
  swing2LimitDegrees=45
  swing2Motion=Free
  twistLimitDegrees=45
  twistMotion=Free
 }
 {
  bone1Name=neck_01
  bone2Name=spine_05
  swing1LimitDegrees=45
  swing1Motion=Free
  swing2LimitDegrees=45
  swing2Motion=Free
  twistLimitDegrees=45
  twistMotion=Free
 }
 {
  bone1Name=neck_02
  bone2Name=neck_01
  swing1LimitDegrees=45
  swing1Motion=Free
  swing2LimitDegrees=45
  swing2Motion=Free
  twistLimitDegrees=45
  twistMotion=Free
 }
 {
  bone1Name=clavicle_l
  bone2Name=spine_05
  swing1LimitDegrees=45
  swing1Motion=Free
  swing2LimitDegrees=45
  swing2Motion=Free
  twistLimitDegrees=45
  twistMotion=Free
 }
 {
  bone1Name=upperarm_l
  bone2Name=clavicle_l
  swing1LimitDegrees=45
  swing1Motion=Free
  swing2LimitDegrees=45
  swing2Motion=Free
  twistLimitDegrees=45
  twistMotion=Free
 }
 {
  bone1Name=lowerarm_l
  bone2Name=upperarm_l
  swing1LimitDegrees=45
  swing1Motion=Free
  swing2LimitDegrees=45
  swing2Motion=Free
  twistLimitDegrees=45
  twistMotion=Free
 }
 {
  bone1Name=clavicle_r
  bone2Name=spine_05
  swing1LimitDegrees=45
  swing1Motion=Free
  swing2LimitDegrees=45
  swing2Motion=Free
  twistLimitDegrees=45
  twistMotion=Free
 }
 {
  bone1Name=upperarm_r
  bone2Name=clavicle_r
  swing1LimitDegrees=45
  swing1Motion=Free
  swing2LimitDegrees=45
  swing2Motion=Free
  twistLimitDegrees=45
  twistMotion=Free
 }
 {
  bone1Name=lowerarm_r
  bone2Name=upperarm_r
  swing1LimitDegrees=45
  swing1Motion=Free
  swing2LimitDegrees=45
  swing2Motion=Free
  twistLimitDegrees=45
  twistMotion=Free
 }
 {
  bone1Name=upperarm_twist_01_r
  bone2Name=upperarm_r
  swing1LimitDegrees=45
  swing1Motion=Free
  swing2LimitDegrees=45
  swing2Motion=Free
  twistLimitDegrees=45
  twistMotion=Free
 }
 {
  bone1Name=upperarm_twist_01_l
  bone2Name=upperarm_l
  swing1LimitDegrees=45
  swing1Motion=Free
  swing2LimitDegrees=45
  swing2Motion=Free
  twistLimitDegrees=45
  twistMotion=Free
 }
]
