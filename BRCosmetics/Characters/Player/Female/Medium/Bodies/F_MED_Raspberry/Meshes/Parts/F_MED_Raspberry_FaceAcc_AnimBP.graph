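# Post-process style AnimBlueprint that simulates the hair braids on the female Raspberry character's FaceAcc mesh.
# Copied parent and body poses are blended, braid roots aim at pelvis offsets, and the cached result feeds front and back RigidBody simulations.
# Bools select plain physics, a backpack variant constraining side braids to back-braid rotation, or a similar upward-gravity variant for skydiving in the front end.
# Computed GravityOverride and ADS hair velocity never reach the AnimGraph.
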AnimBlueprint /BRCosmetics/Characters/Player/Female/Medium/Bodies/F_MED_Raspberry/Meshes/Parts/F_MED_Raspberry_FaceAcc_AnimBP
target_skeleton=&/F_MED_Raspberry_FaceAcc_Skeleton
graphs [
 AnimGraph {nodes [
   n1 = "Transform (Modify) Bone - Bone: dyn_l_back_braid_MSTR"!(ComponentPose, bAlphaBoolEnabled:bool=True, Translation="0,0,0", Rotation="0,10,0", Scale="1,1,1") -> (Pose) {
    alphaInputType=Bool
    rotationMode=BMM_Additive
    rotationSpace=BCS_BoneSpace
    scaleMode=BMM_Additive
    scaleSpace=BCS_BoneSpace
    translationMode=BMM_Additive
    translationSpace=BCS_BoneSpace
    alphaBoolBlend {blendOption=Linear}
    boneToModify {boneName=dyn_l_back_braid_MSTR}
    rotation {yaw=10}
   }
   "Transform (Modify) Bone - Bone: dyn_l_back_braid_04"!(ComponentPose=@n1, bAlphaBoolEnabled:bool=True, Translation="0,0,0", Rotation="0,-20,0", Scale="1,1,1") -> (Pose) {
    alphaInputType=Bool
    rotationMode=BMM_Additive
    rotationSpace=BCS_BoneSpace
    scaleMode=BMM_Additive
    scaleSpace=BCS_BoneSpace
    translationMode=BMM_Additive
    translationSpace=BCS_BoneSpace
    alphaBoolBlend {blendOption=Linear}
    boneToModify {boneName=dyn_l_back_braid_04}
    rotation {yaw=-20}
   }
   n2 = RigidBody!(
    ComponentPose=Local To Component(LocalPose=Use cached pose 'PreSim'())
    Alpha
    OverrideWorldGravity:Vector="0,0,-980"
    ExternalForce
    ComponentLinearAccScale:Vector="0.4,0.4,0.5"
   ) -> (Pose) {
    bClampLinearTranslationLimitToRefPose=true
    bOverrideWorldGravity=true
    overridePhysicsAsset=&/F_MED_Raspberry_FaceAcc_Front_Physics
    simulationSpace=BaseBoneSpace
    alphaBoolBlend {blendOption=Linear}
    baseBoneRef {boneName=spine_05}
    componentAppliedLinearAccClamp=<2500 2500 2500>
    componentLinearAccScale=<0.4 0.4 0.5>
    componentLinearVelScale=<0.3 0.3 0.3>
    overrideWorldGravity {z=-980}
    simSpaceSettings {
     maxAngularAcceleration=1000
     maxAngularVelocity=3
     maxLinearAcceleration=700
     maxLinearVelocity=500
     velocityScaleZ=0
     worldAlpha=0.25
     externalLinearDragV=<0.2 0.2>
    }
   }
   n3 = RigidBody(ComponentPose=@n2, OverrideWorldGravity="0,0,-980", ComponentLinearAccScale="0.4,0.4,0.5") {
    bClampLinearTranslationLimitToRefPose=true
    bOverrideWorldGravity=true
    overridePhysicsAsset=&/F_MED_Raspberry_FaceAcc_Back_Physics
    simulationSpace=BaseBoneSpace
    alphaBoolBlend {blendOption=Linear}
    baseBoneRef {boneName=spine_05}
    componentAppliedLinearAccClamp=<2500 2500 2500>
    componentLinearAccScale=<0.4 0.4 0.5>
    componentLinearVelScale=<0.3 0.3 0.3>
    overrideWorldGravity {z=-980}
    simSpaceSettings {
     maxAngularAcceleration=1000
     maxAngularVelocity=3
     maxLinearAcceleration=700
     maxLinearVelocity=500
     velocityScaleZ=0
     worldAlpha=0.25
     externalLinearDragV=<0.2 0.2>
    }
   }
   n4 = Component To Local(ComponentPose=@n3)
   n5 = RigidBody(
    ComponentPose=Local To Component(LocalPose=Use cached pose 'PreSim'())
    OverrideWorldGravity="0,0,-980"
    ComponentLinearAccScale="0.35,0.35,0.35"
   ) {
    bClampLinearTranslationLimitToRefPose=true
    bOverrideWorldGravity=true
    overridePhysicsAsset=&F_MED_Raspberry_FaceAcc_Front_Physics
    simulationSpace=BaseBoneSpace
    alphaBoolBlend {blendOption=Linear}
    baseBoneRef {boneName=spine_05}
    componentAppliedLinearAccClamp=<2500 2500 2500>
    componentLinearAccScale=<0.4 0.4 0.5>
    componentLinearVelScale=<0.3 0.3 0.3>
    overrideWorldGravity {z=-980}
    simSpaceSettings {
     maxAngularAcceleration=1000
     maxAngularVelocity=3
     maxLinearAcceleration=700
     maxLinearVelocity=500
     velocityScaleZ=0
     worldAlpha=0.5
     externalLinearDragV=<0.2 0.2>
    }
   }
   n6 = RigidBody(ComponentPose=@n5, OverrideWorldGravity="0,0,-980", ComponentLinearAccScale="0,0,0") {
    bClampLinearTranslationLimitToRefPose=true
    bOverrideWorldGravity=true
    overridePhysicsAsset=&F_MED_Raspberry_FaceAcc_Back_Physics
    simulationSpace=BaseBoneSpace
    alphaBoolBlend {blendOption=Linear}
    baseBoneRef {boneName=spine_05}
    componentAppliedLinearAccClamp=<2500 2500 2500>
    componentLinearAccScale=<0.4 0.4 0.5>
    componentLinearVelScale=<0.3 0.3 0.3>
    overrideWorldGravity {z=-980}
    simSpaceSettings {
     maxAngularAcceleration=1000
     maxAngularVelocity=3
     maxLinearAcceleration=700
     maxLinearVelocity=500
     velocityScaleZ=0
     worldAlpha=0.5
     externalLinearDragV=<0.2 0.2>
    }
   }
   n7 = "Transform (Modify) Bone - Bone: dyn_l_side_b_braid_MSTR"!(ComponentPose=@n6, bAlphaBoolEnabled:bool=True, Translation="0,0,0", Rotation="0,-5,0", Scale="1,1,1") -> (Pose) {
    alphaInputType=Bool
    rotationMode=BMM_Additive
    rotationSpace=BCS_BoneSpace
    scaleMode=BMM_Additive
    scaleSpace=BCS_BoneSpace
    translationMode=BMM_Additive
    translationSpace=BCS_BoneSpace
    alphaBoolBlend {blendOption=Linear}
    boneToModify {boneName=dyn_l_side_b_braid_MSTR}
    rotation {yaw=-5}
   }
   n8 = "Transform (Modify) Bone - Bone: dyn_r_side_b_braid_MSTR"!(ComponentPose=@n7, bAlphaBoolEnabled:bool=True, Translation="0,0,0", Rotation="0,5,0", Scale="1,1,1") -> (Pose) {
    alphaInputType=Bool
    rotationMode=BMM_Additive
    rotationSpace=BCS_BoneSpace
    scaleMode=BMM_Additive
    scaleSpace=BCS_BoneSpace
    translationMode=BMM_Additive
    translationSpace=BCS_BoneSpace
    alphaBoolBlend {blendOption=Linear}
    boneToModify {boneName=dyn_r_side_b_braid_MSTR}
    rotation {yaw=5}
   }
   n9 = "Constraint - dyn_r_side_b_braid_01 (R)"!(ComponentPose=@n8, bAlphaBoolEnabled:bool=True, ConstraintWeights_0=1) -> (Pose) {
    alphaInputType=Bool
    alphaBoolBlend {blendInTime=0.1, blendOption=Linear, blendOutTime=0.1}
    boneToModify {boneName=dyn_r_side_b_braid_01}
    constraintSetup [{transformType=Rotation, targetBone {boneName=dyn_back_braid_01}}]
    constraintWeights [1]
   }
   n10 = "Constraint - dyn_r_side_b_braid_02 (R)"!(ComponentPose=@n9, bAlphaBoolEnabled:bool=True, ConstraintWeights_0=1) -> (Pose) {
    alphaInputType=Bool
    alphaBoolBlend {blendInTime=0.1, blendOption=Linear, blendOutTime=0.1}
    boneToModify {boneName=dyn_r_side_b_braid_02}
    constraintSetup [{transformType=Rotation, targetBone {boneName=dyn_back_braid_02}}]
    constraintWeights [1]
   }
   n11 = "Constraint - dyn_r_side_b_braid_03 (R)"!(ComponentPose=@n10, bAlphaBoolEnabled:bool=True, ConstraintWeights_0=1) -> (Pose) {
    alphaInputType=Bool
    alphaBoolBlend {blendInTime=0.1, blendOption=Linear, blendOutTime=0.1}
    boneToModify {boneName=dyn_r_side_b_braid_03}
    constraintSetup [{transformType=Rotation, targetBone {boneName=dyn_back_braid_03}}]
    constraintWeights [1]
   }
   n12 = "Constraint - dyn_r_side_b_braid_04 (R)"!(ComponentPose=@n11, bAlphaBoolEnabled:bool=True, ConstraintWeights_0=1) -> (Pose) {
    alphaInputType=Bool
    alphaBoolBlend {blendInTime=0.1, blendOption=Linear, blendOutTime=0.1}
    boneToModify {boneName=dyn_r_side_b_braid_04}
    constraintSetup [{transformType=Rotation, targetBone {boneName=dyn_back_braid_04}}]
    constraintWeights [1]
   }
   n13 = "Constraint - dyn_r_side_b_braid_05 (R)"!(ComponentPose=@n12, bAlphaBoolEnabled:bool=True, ConstraintWeights_0=1) -> (Pose) {
    alphaInputType=Bool
    alphaBoolBlend {blendInTime=0.1, blendOption=Linear, blendOutTime=0.1}
    boneToModify {boneName=dyn_r_side_b_braid_05}
    constraintSetup [{transformType=Rotation, targetBone {boneName=dyn_back_braid_04}}]
    constraintWeights [1]
   }
   n14 = "Constraint - dyn_l_side_b_braid_01 (R)"!(ComponentPose=@n13, bAlphaBoolEnabled:bool=True, ConstraintWeights_0=1) -> (Pose) {
    alphaInputType=Bool
    alphaBoolBlend {blendInTime=0.1, blendOption=Linear, blendOutTime=0.1}
    boneToModify {boneName=dyn_l_side_b_braid_01}
    constraintSetup [{transformType=Rotation, targetBone {boneName=dyn_back_braid_01}}]
    constraintWeights [1]
   }
   n15 = "Constraint - dyn_l_side_b_braid_02 (R)"!(ComponentPose=@n14, bAlphaBoolEnabled:bool=True, ConstraintWeights_0=1) -> (Pose) {
    alphaInputType=Bool
    alphaBoolBlend {blendInTime=0.1, blendOption=Linear, blendOutTime=0.1}
    boneToModify {boneName=dyn_l_side_b_braid_02}
    constraintSetup [{transformType=Rotation, targetBone {boneName=dyn_back_braid_02}}]
    constraintWeights [1]
   }
   n16 = "Constraint - dyn_l_side_b_braid_03 (R)"!(ComponentPose=@n15, bAlphaBoolEnabled:bool=True, ConstraintWeights_0=1) -> (Pose) {
    alphaInputType=Bool
    alphaBoolBlend {blendInTime=0.1, blendOption=Linear, blendOutTime=0.1}
    boneToModify {boneName=dyn_l_side_b_braid_03}
    constraintSetup [{transformType=Rotation, targetBone {boneName=dyn_back_braid_03}}]
    constraintWeights [1]
   }
   n17 = "Constraint - dyn_l_side_b_braid_04 (R)"!(ComponentPose=@n16, bAlphaBoolEnabled:bool=True, ConstraintWeights_0=1) -> (Pose) {
    alphaInputType=Bool
    alphaBoolBlend {blendInTime=0.1, blendOption=Linear, blendOutTime=0.1}
    boneToModify {boneName=dyn_l_side_b_braid_04}
    constraintSetup [{transformType=Rotation, targetBone {boneName=dyn_back_braid_04}}]
    constraintWeights [1]
   }
   n18 = "Constraint - dyn_l_side_b_braid_05 (R)"!(ComponentPose=@n17, bAlphaBoolEnabled:bool=True, ConstraintWeights_0=1) -> (Pose) {
    alphaInputType=Bool
    alphaBoolBlend {blendInTime=0.1, blendOption=Linear, blendOutTime=0.1}
    boneToModify {boneName=dyn_l_side_b_braid_05}
    constraintSetup [{transformType=Rotation, targetBone {boneName=dyn_back_braid_04}}]
    constraintWeights [1]
   }
   n19 = Component To Local(ComponentPose=@n18)
   n20 = Get isWearingBackpack() -> (isWearingBackpack:bool)
   n21 = RigidBody(
    ComponentPose=Local To Component(LocalPose=Use cached pose 'PreSim'())
    OverrideWorldGravity="0,0,350"
    ComponentLinearAccScale="0.35,0.35,0.35"
   ) {
    bClampLinearTranslationLimitToRefPose=true
    bOverrideWorldGravity=true
    overridePhysicsAsset=&F_MED_Raspberry_FaceAcc_Front_Physics
    simulationSpace=BaseBoneSpace
    alphaBoolBlend {blendOption=Linear}
    baseBoneRef {boneName=spine_05}
    componentAppliedLinearAccClamp=<2500 2500 2500>
    componentLinearAccScale=<0.4 0.4 0.5>
    componentLinearVelScale=<0.3 0.3 0.3>
    overrideWorldGravity {z=-980}
    simSpaceSettings {
     maxAngularAcceleration=1000
     maxAngularVelocity=3
     maxLinearAcceleration=700
     maxLinearVelocity=500
     velocityScaleZ=0
     worldAlpha=0.1
     externalLinearDragV=<0.2 0.2>
    }
   }
   n22 = RigidBody(ComponentPose=@n21, OverrideWorldGravity="0,0,350", ComponentLinearAccScale="0,0,0") {
    bClampLinearTranslationLimitToRefPose=true
    bOverrideWorldGravity=true
    overridePhysicsAsset=&F_MED_Raspberry_FaceAcc_Back_Physics
    simulationSpace=BaseBoneSpace
    alphaBoolBlend {blendOption=Linear}
    baseBoneRef {boneName=spine_05}
    componentAppliedLinearAccClamp=<2500 2500 2500>
    componentLinearAccScale=<0.4 0.4 0.5>
    componentLinearVelScale=<0.3 0.3 0.3>
    overrideWorldGravity {z=-980}
    simSpaceSettings {
     maxAngularAcceleration=1000
     maxAngularVelocity=3
     maxLinearAcceleration=700
     maxLinearVelocity=500
     velocityScaleZ=0
     worldAlpha=0.1
     externalLinearDragV=<0.2 0.2>
    }
   }
   n23 = "Transform (Modify) Bone - Bone: dyn_l_side_b_braid_MSTR"(ComponentPose=@n22, bAlphaBoolEnabled=True, Translation="0,0,0", Rotation="0,-5,0", Scale="1,1,1") {
    alphaInputType=Bool
    rotationMode=BMM_Additive
    rotationSpace=BCS_BoneSpace
    scaleMode=BMM_Additive
    scaleSpace=BCS_BoneSpace
    translationMode=BMM_Additive
    translationSpace=BCS_BoneSpace
    alphaBoolBlend {blendOption=Linear}
    boneToModify {boneName=dyn_l_side_b_braid_MSTR}
    rotation {yaw=-5}
   }
   n24 = "Transform (Modify) Bone - Bone: dyn_r_side_b_braid_MSTR"(ComponentPose=@n23, bAlphaBoolEnabled=True, Translation="0,0,0", Rotation="0,5,0", Scale="1,1,1") {
    alphaInputType=Bool
    rotationMode=BMM_Additive
    rotationSpace=BCS_BoneSpace
    scaleMode=BMM_Additive
    scaleSpace=BCS_BoneSpace
    translationMode=BMM_Additive
    translationSpace=BCS_BoneSpace
    alphaBoolBlend {blendOption=Linear}
    boneToModify {boneName=dyn_r_side_b_braid_MSTR}
    rotation {yaw=5}
   }
   n25 = "Constraint - dyn_r_side_b_braid_01 (R)"(ComponentPose=@n24, bAlphaBoolEnabled=True, ConstraintWeights_0=1) {
    alphaInputType=Bool
    alphaBoolBlend {blendInTime=0.1, blendOption=Linear, blendOutTime=0.1}
    boneToModify {boneName=dyn_r_side_b_braid_01}
    constraintSetup [{transformType=Rotation, targetBone {boneName=dyn_back_braid_01}}]
    constraintWeights [1]
   }
   n26 = "Constraint - dyn_r_side_b_braid_02 (R)"(ComponentPose=@n25, bAlphaBoolEnabled=True, ConstraintWeights_0=1) {
    alphaInputType=Bool
    alphaBoolBlend {blendInTime=0.1, blendOption=Linear, blendOutTime=0.1}
    boneToModify {boneName=dyn_r_side_b_braid_02}
    constraintSetup [{transformType=Rotation, targetBone {boneName=dyn_back_braid_02}}]
    constraintWeights [1]
   }
   n27 = "Constraint - dyn_r_side_b_braid_03 (R)"(ComponentPose=@n26, bAlphaBoolEnabled=True, ConstraintWeights_0=1) {
    alphaInputType=Bool
    alphaBoolBlend {blendInTime=0.1, blendOption=Linear, blendOutTime=0.1}
    boneToModify {boneName=dyn_r_side_b_braid_03}
    constraintSetup [{transformType=Rotation, targetBone {boneName=dyn_back_braid_03}}]
    constraintWeights [1]
   }
   n28 = "Constraint - dyn_r_side_b_braid_04 (R)"(ComponentPose=@n27, bAlphaBoolEnabled=True, ConstraintWeights_0=1) {
    alphaInputType=Bool
    alphaBoolBlend {blendInTime=0.1, blendOption=Linear, blendOutTime=0.1}
    boneToModify {boneName=dyn_r_side_b_braid_04}
    constraintSetup [{transformType=Rotation, targetBone {boneName=dyn_back_braid_04}}]
    constraintWeights [1]
   }
   n29 = "Constraint - dyn_r_side_b_braid_05 (R)"(ComponentPose=@n28, bAlphaBoolEnabled=True, ConstraintWeights_0=1) {
    alphaInputType=Bool
    alphaBoolBlend {blendInTime=0.1, blendOption=Linear, blendOutTime=0.1}
    boneToModify {boneName=dyn_r_side_b_braid_05}
    constraintSetup [{transformType=Rotation, targetBone {boneName=dyn_back_braid_04}}]
    constraintWeights [1]
   }
   n30 = "Constraint - dyn_l_side_b_braid_01 (R)"(ComponentPose=@n29, bAlphaBoolEnabled=True, ConstraintWeights_0=1) {
    alphaInputType=Bool
    alphaBoolBlend {blendInTime=0.1, blendOption=Linear, blendOutTime=0.1}
    boneToModify {boneName=dyn_l_side_b_braid_01}
    constraintSetup [{transformType=Rotation, targetBone {boneName=dyn_back_braid_01}}]
    constraintWeights [1]
   }
   n31 = "Constraint - dyn_l_side_b_braid_02 (R)"(ComponentPose=@n30, bAlphaBoolEnabled=True, ConstraintWeights_0=1) {
    alphaInputType=Bool
    alphaBoolBlend {blendInTime=0.1, blendOption=Linear, blendOutTime=0.1}
    boneToModify {boneName=dyn_l_side_b_braid_02}
    constraintSetup [{transformType=Rotation, targetBone {boneName=dyn_back_braid_02}}]
    constraintWeights [1]
   }
   n32 = "Constraint - dyn_l_side_b_braid_03 (R)"(ComponentPose=@n31, bAlphaBoolEnabled=True, ConstraintWeights_0=1) {
    alphaInputType=Bool
    alphaBoolBlend {blendInTime=0.1, blendOption=Linear, blendOutTime=0.1}
    boneToModify {boneName=dyn_l_side_b_braid_03}
    constraintSetup [{transformType=Rotation, targetBone {boneName=dyn_back_braid_03}}]
    constraintWeights [1]
   }
   n33 = "Constraint - dyn_l_side_b_braid_04 (R)"(ComponentPose=@n32, bAlphaBoolEnabled=True, ConstraintWeights_0=1) {
    alphaInputType=Bool
    alphaBoolBlend {blendInTime=0.1, blendOption=Linear, blendOutTime=0.1}
    boneToModify {boneName=dyn_l_side_b_braid_04}
    constraintSetup [{transformType=Rotation, targetBone {boneName=dyn_back_braid_04}}]
    constraintWeights [1]
   }
   n34 = "Constraint - dyn_l_side_b_braid_05 (R)"(ComponentPose=@n33, bAlphaBoolEnabled=True, ConstraintWeights_0=1) {
    alphaInputType=Bool
    alphaBoolBlend {blendInTime=0.1, blendOption=Linear, blendOutTime=0.1}
    boneToModify {boneName=dyn_l_side_b_braid_05}
    constraintSetup [{transformType=Rotation, targetBone {boneName=dyn_back_braid_04}}]
    constraintWeights [1]
   }
   n35 = Get bIsSkyDiving()
   n36 = Get bIsInFrontEnd()
   Output Pose(Result=Blend Poses by bool(
     BlendPose_0=Component To Local(ComponentPose=@n34)
     BlendPose_1=Blend Poses by bool(BlendPose_0=@n19, BlendPose_1=@n4, bActiveValue=@n20) {blendType=Linear}
     bActiveValue=AND Boolean(A=@n35, B=@n36)
    ) {blendType=Linear}) {layerGroup=None}
   n37 = Copy Pose From Mesh(SourceMeshComponent=Get Body_Mesh())
   n38 = Copy Pose From Mesh() {bCopyCurves=true, bUseAttachedParent=true}
   n39 = Look At - Bone: dyn_r_side_b_braid_MSTR(
    ComponentPose=Local To Component(LocalPose=Layered blend per bone(BasePose=@n38, BlendPoses_0=@n37) {layerSetup [{branchFilters [{boneName=upperarm_l}, {boneName=upperarm_r}]}]})
    Alpha=1
   ) {
    alphaBoolBlend {blendOption=Linear}
    boneToModify {boneName=dyn_r_side_b_braid_MSTR}
    lookAtLocation=<0 -30 15>
    lookAtTarget {boneReference {boneName=pelvis}}
    lookAt_Axis {axis=<1 0>}
   }
   n40 = Look At - Bone: dyn_l_side_b_braid_MSTR(ComponentPose=@n39, Alpha=1) {
    alphaBoolBlend {blendOption=Linear}
    boneToModify {boneName=dyn_l_side_b_braid_MSTR}
    lookAtLocation=<0 -30 -15>
    lookAtTarget {boneReference {boneName=pelvis}}
    lookAt_Axis {axis=<1 0>}
   }
   n41 = Look At - Bone: dyn_r_side_a_braid_MSTR(ComponentPose=@n40, Alpha=1) {
    alphaBoolBlend {blendOption=Linear}
    boneToModify {boneName=dyn_r_side_a_braid_MSTR}
    lookAtLocation=<0 15 30>
    lookAtTarget {boneReference {boneName=pelvis}}
    lookAt_Axis {axis=<1 0>}
   }
   n42 = Look At - Bone: dyn_l_side_a_braid_MSTR(ComponentPose=@n41, Alpha=1) {
    alphaBoolBlend {blendOption=Linear}
    boneToModify {boneName=dyn_l_side_a_braid_MSTR}
    lookAtLocation=<0 15 -30>
    lookAtTarget {boneReference {boneName=pelvis}}
    lookAt_Axis {axis=<1 0>}
   }
   n43 = Look At - Bone: dyn_r_fr_braid_MSTR(ComponentPose=@n42, Alpha=1) {
    alphaBoolBlend {blendOption=Linear}
    boneToModify {boneName=dyn_r_fr_braid_MSTR}
    lookAtLocation=<0 15 10>
    lookAtTarget {boneReference {boneName=pelvis}}
    lookAt_Axis {axis=<1 0>}
   }
   n44 = Look At - Bone: dyn_l_fr_braid_MSTR(ComponentPose=@n43, Alpha=1) {
    alphaBoolBlend {blendOption=Linear}
    boneToModify {boneName=dyn_l_fr_braid_MSTR}
    lookAtLocation=<0 15 -15>
    lookAtTarget {boneReference {boneName=pelvis}}
    lookAt_Axis {axis=<1 0>}
   }
   Save cached pose 'PreSim'(Pose=Component To Local(ComponentPose=@n44)) {cacheName=PreSim}
  ]}
 EventGraph {nodes [
   Event BlueprintUpdateAnimation()
   > n1 = Sequence!(execute) -> (then_0, then_1, then_2:Exec)
   n2 = Gravity Override State Function DEPRECATED(
    execute=@n1.then_1
    CharacterPartAnimInstance=Self-Reference()
    GameTimeInSeconds=GetGameTimeInSeconds()
    GravityOverrideParams=Get Gravity Override Params()
   )
   > Set GravityOverride(GravityOverride=@n2)
   n3 = Get bIsCrouching()
   n4 = Get bIsTargeting()
   n5 = Get bIsPlayingMeleeAnim()
   n6 = Get bIsSkyDiving()
   Set ADS Hair Velocity(
    execute=@n1.then_2
    ADS Hair Velocity:Vector=SelectVector(A="0.15,0.15,0.25", B="0.38,0.38,0.38", bPickA=OR Boolean!(A=@n4, B=@n5, C:bool=@n6, D:bool=@n3) -> (ReturnValue))
   ) -> (Output_Get:Vector)
   Set Body_Mesh(execute=[@n8.then, @n1.then_0], Body_Mesh=GetOwnerSkeletalMeshComponent())
   Event OnCharacterCustomizationLoaded()
   n7 = CheckForBackbling()
   > Set isWearingBackpack(isWearingBackpack:bool=OR Boolean!(A=@n7, B=@n7.bIsBackblingACape, C=@n7.bIsBackblingLong) -> (ReturnValue)) -> (Output_Get:bool)
   n8 = Event BlueprintInitializeAnimation()
   SelectVector(A="0.15,0.15,0.25", B="0.3,0.3,0.45", bPickA=false)
  ]}
]
variables [
 ADS Hair Velocity {type=Vector}
 backblingEquippedGravity {type=Vector}
 Body_Mesh {type=SkeletalMeshComponent}
 General Gravity Override {type=Vector}
 Gravity Override Params {
  type=GravityOverrideParamsStruct
  default {skydive_FloatingDown {gravityOverride=<0 -150 500>}, skydive_GoingUp {gravityOverride=<0 -800 500>}}
 }
 gravityFinal {type=Vector}
 GravityOverride {type=Vector}
 isWearingBackpack {type=boolean}
]
